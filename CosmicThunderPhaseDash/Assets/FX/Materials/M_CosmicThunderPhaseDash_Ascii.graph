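# Translucent unlit Niagara sprite material for Cosmic Thunder Phase Dash that renders flickering, glitched ASCII characters.
# Custom HLSL holds an 8x8 bitmap font and hashes scaled time with particle random to pick each glyph, sampled at clean and noise-displaced UVs.
# Both samples form opacity with particle alpha and squared near-camera fade; the clean glyph feeds hue-shifted emissive over particle color.
# Dynamic parameters drive time scale, glitch, hue shift and distortion.
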Material /CosmicThunderPhaseDash/Assets/FX/Materials/M_CosmicThunderPhaseDash_Ascii
graph {
 properties {
  bAutomaticallySetUsageInEditor=false
  bUsedWithNiagaraSprites=true
  blendMode=BLEND_Translucent
  shadingModel=MSM_Unlit
  usedShadingModels=Unlit
 }
 nodes [
  n1 = TextureSample(UVs=Add(A=ParticleRandom(), B=TextureCoordinate() {uTiling=0.1})) {texture=/Game/Effects/Fort_Effects/Effects/Atlas/Textures/T_Atlas_DigitalNoise_02}
  n2 = Multiply(
   A=Subtract(A=@n1.RGB) {constB=0.5}
   B=DynamicParameter() {defaultValue=<rgba 1 0.1 0 0>, paramNames [Time Scale, Glitch, HueShift, Distortion]}.Distortion
  )
  AppendVector(A=@n1.R, B=@n2)
  n3 = DynamicParameter() {defaultValue=<rgba 1 0.1 0 0>, paramNames [Time Scale, Glitch, HueShift, Distortion]}
  MaterialExpressionMultiply_1 = Multiply(
   A=DynamicParameter() {defaultValue=<rgba 1 0.1 0 0>, paramNames [Time Scale, Glitch, HueShift, Distortion]}.Time Scale
   B=Multiply(A=Time()) {constB=5}
  )
  MaterialExpressionTextureCoordinate_0 = TextureCoordinate()
  MaterialExpressionParticleRandom_0 = ParticleRandom()
  MaterialExpressionConstant_0 = Constant() {r=1}
  n4 = Custom!(
   Time=@MaterialExpressionMultiply_1
   InUV=@MaterialExpressionTextureCoordinate_0
   Seed=@MaterialExpressionParticleRandom_0
   Scale=@MaterialExpressionConstant_0
  ) {
   code="static const uint2 FONT[95] = {\r\n    uint2(0x00000000u, 0x00000000u),  // 32 ' '\r\n    uint2(0x183C3C18u, 0x00180018u),  // 33 '!'\r\n    uint2(0x00003636u, 0x00000000u),  // 34 '\"'\r\n    uint2(0x367F3636u, 0x0036367Fu),  // 35 '##'\r\n    uint2(0x1E033E0Cu, 0x000C1F30u),  // 36 '$'\r\n    uint2(0x18336300u, 0x0063660Cu),  // 37 '%'\r\n    uint2(0x6E1C361Cu, 0x006E333Bu),  // 38 '&&'\r\n    uint2(0x00030606u, 0x00000000u),  // 39 '''\r\n    uint2(0x06060C18u, 0x00180C06u),  // 40 '('\r\n    uint2(0x18180C06u, 0x00060C18u),  // 41 ')'\r\n    uint2(0xFF3C6600u, 0x0000663Cu),  // 42 '*'\r\n    uint2(0x3F0C0C00u, 0x00000C0Cu),  // 43 '+'\r\n    uint2(0x00000000u, 0x060C0C00u),  // 44 ','\r\n    uint2(0x3F000000u, 0x00000000u),  // 45 '-'\r\n    uint2(0x00000000u, 0x000C0C00u),  // 46 '.'\r\n    uint2(0x0C183060u, 0x00010306u),  // 47 '/'\r\n    uint2(0x7B73633Eu, 0x003E676Fu),  // 48 '0'\r\n    uint2(0x0C0C0E0Cu, 0x003F0C0Cu),  // 49 '1'\r\n    uint2(0x1C30331Eu, 0x003F3306u),  // 50 '2'\r\n    uint2(0x1C30331Eu, 0x001E3330u),  // 51 '3'\r\n    uint2(0x33363C38u, 0x0078307Fu),  // 52 '4'\r\n    uint2(0x301F033Fu, 0x001E3330u),  // 53 '5'\r\n    uint2(0x1F03061Cu, 0x001E3333u),  // 54 '6'\r\n    uint2(0x1830333Fu, 0x000C0C0Cu),  // 55 '7'\r\n    uint2(0x1E33331Eu, 0x001E3333u),  // 56 '8'\r\n    uint2(0x3E33331Eu, 0x000E1830u),  // 57 '9'\r\n    uint2(0x000C0C00u, 0x000C0C00u),  // 58 ':'\r\n    uint2(0x000C0C00u, 0x060C0C00u),  // 59 ';'\r\n    uint2(0x03060C18u, 0x00180C06u),  // 60 '<'\r\n    uint2(0x003F0000u, 0x00003F00u),  // 61 '='\r\n    uint2(0x30180C06u, 0x00060C18u),  // 62 '>'\r\n    uint2(0x1830331Eu, 0x000C000Cu),  // 63 '?'\r\n    uint2(0x7B7B633Eu, 0x001E037Bu),  // 64 '@'\r\n    uint2(0x33331E0Cu, 0x0033333Fu),  // 65 'A'\r\n    uint2(0x3E66663Fu, 0x003F6666u),  // 66 'B'\r\n    uint2(0x0303663Cu, 0x003C6603u),  // 67 'C'\r\n    uint2(0x6666361Fu, 0x001F3666u),  // 68 'D'\r\n    uint2(0x1E16467Fu, 0x007F4616u),  // 69 'E'\r\n    uint2(0x1E16467Fu, 0x000F0616u),  // 70 'F'\r\n    uint2(0x0303663Cu, 0x007C6673u),  // 71 'G'\r\n    uint2(0x3F333333u, 0x00333333u),  // 72 'H'\r\n    uint2(0x0C0C0C1Eu, 0x001E0C0Cu),  // 73 'I'\r\n    uint2(0x30303078u, 0x001E3333u),  // 74 'J'\r\n    uint2(0x1E366667u, 0x00676636u),  // 75 'K'\r\n    uint2(0x0606060Fu, 0x007F6646u),  // 76 'L'\r\n    uint2(0x7F7F7763u, 0x0063636Bu),  // 77 'M'\r\n    uint2(0x7B6F6763u, 0x00636373u),  // 78 'N'\r\n    uint2(0x6363361Cu, 0x001C3663u),  // 79 'O'\r\n    uint2(0x3E66663Fu, 0x000F0606u),  // 80 'P'\r\n    uint2(0x3333331Eu, 0x00381E3Bu),  // 81 'Q'\r\n    uint2(0x3E66663Fu, 0x00676636u),  // 82 'R'\r\n    uint2(0x0E07331Eu, 0x001E3338u),  // 83 'S'\r\n    uint2(0x0C0C2D3Fu, 0x001E0C0Cu),  // 84 'T'\r\n    uint2(0x33333333u, 0x003F3333u),  // 85 'U'\r\n    uint2(0x33333333u, 0x000C1E33u),  // 86 'V'\r\n    uint2(0x6B636363u, 0x0063777Fu),  // 87 'W'\r\n    uint2(0x1C366363u, 0x0063361Cu),  // 88 'X'\r\n    uint2(0x1E333333u, 0x001E0C0Cu),  // 89 'Y'\r\n    uint2(0x1831637Fu, 0x007F664Cu),  // 90 'Z'\r\n    uint2(0x0606061Eu, 0x001E0606u),  // 91 '['\r\n    uint2(0x180C0603u, 0x00406030u),  // 92 'backslash'\r\n    uint2(0x1818181Eu, 0x001E1818u),  // 93 ']'\r\n    uint2(0x63361C08u, 0x00000000u),  // 94 '^'\r\n    uint2(0x00000000u, 0xFF000000u),  // 95 '_'\r\n    uint2(0x00180C0Cu, 0x00000000u),  // 96 '`'\r\n    uint2(0x301E0000u, 0x006E333Eu),  // 97 'a'\r\n    uint2(0x3E060607u, 0x003B6666u),  // 98 'b'\r\n    uint2(0x331E0000u, 0x001E3303u),  // 99 'c'\r\n    uint2(0x3E303038u, 0x006E3333u),  // 100 'd'\r\n    uint2(0x331E0000u, 0x001E033Fu),  // 101 'e'\r\n    uint2(0x0F06361Cu, 0x000F0606u),  // 102 'f'\r\n    uint2(0x336E0000u, 0x1F303E33u),  // 103 'g'\r\n    uint2(0x6E360607u, 0x00676666u),  // 104 'h'\r\n    uint2(0x0C0E000Cu, 0x001E0C0Cu),  // 105 'i'\r\n    uint2(0x30300030u, 0x1E333330u),  // 106 'j'\r\n    uint2(0x36660607u, 0x0067361Eu),  // 107 'k'\r\n    uint2(0x0C0C0C0Eu, 0x001E0C0Cu),  // 108 'l'\r\n    uint2(0x7F330000u, 0x00636B7Fu),  // 109 'm'\r\n    uint2(0x331F0000u, 0x00333333u),  // 110 'n'\r\n    uint2(0x331E0000u, 0x001E3333u),  // 111 'o'\r\n    uint2(0x663B0000u, 0x0F063E66u),  // 112 'p'\r\n    uint2(0x336E0000u, 0x78303E33u),  // 113 'q'\r\n    uint2(0x6E3B0000u, 0x000F0666u),  // 114 'r'\r\n    uint2(0x033E0000u, 0x001F301Eu),  // 115 's'\r\n    uint2(0x0C3E0C08u, 0x00182C0Cu),  // 116 't'\r\n    uint2(0x33330000u, 0x006E3333u),  // 117 'u'\r\n    uint2(0x33330000u, 0x000C1E33u),  // 118 'v'\r\n    uint2(0x6B630000u, 0x00367F7Fu),  // 119 'w'\r\n    uint2(0x36630000u, 0x0063361Cu),  // 120 'x'\r\n    uint2(0x33330000u, 0x1F303E33u),  // 121 'y'\r\n    uint2(0x193F0000u, 0x003F260Cu),  // 122 'z'\r\n    uint2(0x070C0C38u, 0x00380C0Cu),  // 123 '{'\r\n    uint2(0x00181818u, 0x00181818u),  // 124 '|'\r\n    uint2(0x380C0C07u, 0x00070C0Cu),  // 125 '}'\r\n    uint2(0x00003B6Eu, 0x00000000u),  // 126 '~'\r\n};\r\n\r\nuint h = (uint)(int)floor(Time) \r\n       ^ (uint)(int)(Seed * 65536.0) * 1664525u;\r\nh ^= h >> 16; h *= 0x7FEB352Du;\r\nh ^= h >> 15; h *= 0x846CA68Bu;\r\nh ^= h >> 16;\r\nuint code = 33u + (h % 94u);  \r\n\r\n\r\nfloat2 uv = (InUV - 0.5) / max(Scale, 0.0001) + 0.5; \r\nuv -= 0.51 / 8.0;  \r\nif (uv.x < 0.0 || uv.y < 0.0 || uv.x >= 1.0 || uv.y >= 1.0) return 0.0;\r\n\r\nuint2 g = FONT[code - 32u];\r\nuint2 p = (uint2)(uv * 8.0);\r\nuint rowBits = ((p.y < 4u) ? (g.x >> (p.y * 8u))\r\n                           : (g.y >> ((p.y - 4u) * 8u))) && 0xFFu;\r\nreturn (float)((rowBits >> p.x) && 1u);"
   outputType=CMOT_Float1
   inputs [
    {inputName=Time, input {expression=&self:MaterialExpressionMultiply_1}}
    {
     inputName=InUV
     input {expression=&self:MaterialExpressionTextureCoordinate_0}
    }
    {
     inputName=Seed
     input {expression=&self:MaterialExpressionParticleRandom_0}
    }
    {inputName=Scale, input {expression=&self:MaterialExpressionConstant_0}}
   ]
  }
  n5 = ParticleColor()
  n6 = Multiply(A=@n4, B=@n5.RGB)
  MaterialFunctionCall(Emissive=Add(
    A=MaterialFunctionCall(Hue Shift Percentage=@n3.HueShift, Texture=@n6) {materialFunction=&HueShift}.Result
    B=@n5.RGB
   )) {materialFunction=&MF_FX_EmissiveAdjustments}
  n7 = ComponentMask(None=@n2) {g=true, r=true}
  n8 = DynamicParameter() {defaultValue=<rgba 1 0.1 0 0>, paramNames [Time Scale, Glitch, HueShift, Distortion]}
  MaterialExpressionSaturate_1 = Saturate(None=Add(A=@MaterialExpressionTextureCoordinate_0, B=Multiply(A=@n7, B=@n8.Glitch)))
  n9 = Custom(
   Time=@MaterialExpressionMultiply_1
   InUV=@MaterialExpressionSaturate_1
   Seed=@MaterialExpressionParticleRandom_0
   Scale=@MaterialExpressionConstant_0
  ) {
   code="static const uint2 FONT[95] = {\r\n    uint2(0x00000000u, 0x00000000u),  // 32 ' '\r\n    uint2(0x183C3C18u, 0x00180018u),  // 33 '!'\r\n    uint2(0x00003636u, 0x00000000u),  // 34 '\"'\r\n    uint2(0x367F3636u, 0x0036367Fu),  // 35 '##'\r\n    uint2(0x1E033E0Cu, 0x000C1F30u),  // 36 '$'\r\n    uint2(0x18336300u, 0x0063660Cu),  // 37 '%'\r\n    uint2(0x6E1C361Cu, 0x006E333Bu),  // 38 '&&'\r\n    uint2(0x00030606u, 0x00000000u),  // 39 '''\r\n    uint2(0x06060C18u, 0x00180C06u),  // 40 '('\r\n    uint2(0x18180C06u, 0x00060C18u),  // 41 ')'\r\n    uint2(0xFF3C6600u, 0x0000663Cu),  // 42 '*'\r\n    uint2(0x3F0C0C00u, 0x00000C0Cu),  // 43 '+'\r\n    uint2(0x00000000u, 0x060C0C00u),  // 44 ','\r\n    uint2(0x3F000000u, 0x00000000u),  // 45 '-'\r\n    uint2(0x00000000u, 0x000C0C00u),  // 46 '.'\r\n    uint2(0x0C183060u, 0x00010306u),  // 47 '/'\r\n    uint2(0x7B73633Eu, 0x003E676Fu),  // 48 '0'\r\n    uint2(0x0C0C0E0Cu, 0x003F0C0Cu),  // 49 '1'\r\n    uint2(0x1C30331Eu, 0x003F3306u),  // 50 '2'\r\n    uint2(0x1C30331Eu, 0x001E3330u),  // 51 '3'\r\n    uint2(0x33363C38u, 0x0078307Fu),  // 52 '4'\r\n    uint2(0x301F033Fu, 0x001E3330u),  // 53 '5'\r\n    uint2(0x1F03061Cu, 0x001E3333u),  // 54 '6'\r\n    uint2(0x1830333Fu, 0x000C0C0Cu),  // 55 '7'\r\n    uint2(0x1E33331Eu, 0x001E3333u),  // 56 '8'\r\n    uint2(0x3E33331Eu, 0x000E1830u),  // 57 '9'\r\n    uint2(0x000C0C00u, 0x000C0C00u),  // 58 ':'\r\n    uint2(0x000C0C00u, 0x060C0C00u),  // 59 ';'\r\n    uint2(0x03060C18u, 0x00180C06u),  // 60 '<'\r\n    uint2(0x003F0000u, 0x00003F00u),  // 61 '='\r\n    uint2(0x30180C06u, 0x00060C18u),  // 62 '>'\r\n    uint2(0x1830331Eu, 0x000C000Cu),  // 63 '?'\r\n    uint2(0x7B7B633Eu, 0x001E037Bu),  // 64 '@'\r\n    uint2(0x33331E0Cu, 0x0033333Fu),  // 65 'A'\r\n    uint2(0x3E66663Fu, 0x003F6666u),  // 66 'B'\r\n    uint2(0x0303663Cu, 0x003C6603u),  // 67 'C'\r\n    uint2(0x6666361Fu, 0x001F3666u),  // 68 'D'\r\n    uint2(0x1E16467Fu, 0x007F4616u),  // 69 'E'\r\n    uint2(0x1E16467Fu, 0x000F0616u),  // 70 'F'\r\n    uint2(0x0303663Cu, 0x007C6673u),  // 71 'G'\r\n    uint2(0x3F333333u, 0x00333333u),  // 72 'H'\r\n    uint2(0x0C0C0C1Eu, 0x001E0C0Cu),  // 73 'I'\r\n    uint2(0x30303078u, 0x001E3333u),  // 74 'J'\r\n    uint2(0x1E366667u, 0x00676636u),  // 75 'K'\r\n    uint2(0x0606060Fu, 0x007F6646u),  // 76 'L'\r\n    uint2(0x7F7F7763u, 0x0063636Bu),  // 77 'M'\r\n    uint2(0x7B6F6763u, 0x00636373u),  // 78 'N'\r\n    uint2(0x6363361Cu, 0x001C3663u),  // 79 'O'\r\n    uint2(0x3E66663Fu, 0x000F0606u),  // 80 'P'\r\n    uint2(0x3333331Eu, 0x00381E3Bu),  // 81 'Q'\r\n    uint2(0x3E66663Fu, 0x00676636u),  // 82 'R'\r\n    uint2(0x0E07331Eu, 0x001E3338u),  // 83 'S'\r\n    uint2(0x0C0C2D3Fu, 0x001E0C0Cu),  // 84 'T'\r\n    uint2(0x33333333u, 0x003F3333u),  // 85 'U'\r\n    uint2(0x33333333u, 0x000C1E33u),  // 86 'V'\r\n    uint2(0x6B636363u, 0x0063777Fu),  // 87 'W'\r\n    uint2(0x1C366363u, 0x0063361Cu),  // 88 'X'\r\n    uint2(0x1E333333u, 0x001E0C0Cu),  // 89 'Y'\r\n    uint2(0x1831637Fu, 0x007F664Cu),  // 90 'Z'\r\n    uint2(0x0606061Eu, 0x001E0606u),  // 91 '['\r\n    uint2(0x180C0603u, 0x00406030u),  // 92 'backslash'\r\n    uint2(0x1818181Eu, 0x001E1818u),  // 93 ']'\r\n    uint2(0x63361C08u, 0x00000000u),  // 94 '^'\r\n    uint2(0x00000000u, 0xFF000000u),  // 95 '_'\r\n    uint2(0x00180C0Cu, 0x00000000u),  // 96 '`'\r\n    uint2(0x301E0000u, 0x006E333Eu),  // 97 'a'\r\n    uint2(0x3E060607u, 0x003B6666u),  // 98 'b'\r\n    uint2(0x331E0000u, 0x001E3303u),  // 99 'c'\r\n    uint2(0x3E303038u, 0x006E3333u),  // 100 'd'\r\n    uint2(0x331E0000u, 0x001E033Fu),  // 101 'e'\r\n    uint2(0x0F06361Cu, 0x000F0606u),  // 102 'f'\r\n    uint2(0x336E0000u, 0x1F303E33u),  // 103 'g'\r\n    uint2(0x6E360607u, 0x00676666u),  // 104 'h'\r\n    uint2(0x0C0E000Cu, 0x001E0C0Cu),  // 105 'i'\r\n    uint2(0x30300030u, 0x1E333330u),  // 106 'j'\r\n    uint2(0x36660607u, 0x0067361Eu),  // 107 'k'\r\n    uint2(0x0C0C0C0Eu, 0x001E0C0Cu),  // 108 'l'\r\n    uint2(0x7F330000u, 0x00636B7Fu),  // 109 'm'\r\n    uint2(0x331F0000u, 0x00333333u),  // 110 'n'\r\n    uint2(0x331E0000u, 0x001E3333u),  // 111 'o'\r\n    uint2(0x663B0000u, 0x0F063E66u),  // 112 'p'\r\n    uint2(0x336E0000u, 0x78303E33u),  // 113 'q'\r\n    uint2(0x6E3B0000u, 0x000F0666u),  // 114 'r'\r\n    uint2(0x033E0000u, 0x001F301Eu),  // 115 's'\r\n    uint2(0x0C3E0C08u, 0x00182C0Cu),  // 116 't'\r\n    uint2(0x33330000u, 0x006E3333u),  // 117 'u'\r\n    uint2(0x33330000u, 0x000C1E33u),  // 118 'v'\r\n    uint2(0x6B630000u, 0x00367F7Fu),  // 119 'w'\r\n    uint2(0x36630000u, 0x0063361Cu),  // 120 'x'\r\n    uint2(0x33330000u, 0x1F303E33u),  // 121 'y'\r\n    uint2(0x193F0000u, 0x003F260Cu),  // 122 'z'\r\n    uint2(0x070C0C38u, 0x00380C0Cu),  // 123 '{'\r\n    uint2(0x00181818u, 0x00181818u),  // 124 '|'\r\n    uint2(0x380C0C07u, 0x00070C0Cu),  // 125 '}'\r\n    uint2(0x00003B6Eu, 0x00000000u),  // 126 '~'\r\n};\r\n\r\nuint h = (uint)(int)floor(Time) \r\n       ^ (uint)(int)(Seed * 65536.0) * 1664525u;\r\nh ^= h >> 16; h *= 0x7FEB352Du;\r\nh ^= h >> 15; h *= 0x846CA68Bu;\r\nh ^= h >> 16;\r\nuint code = 33u + (h % 94u);  \r\n\r\n\r\nfloat2 uv = (InUV - 0.5) / max(Scale, 0.0001) + 0.5; \r\nuv -= 0.51 / 8.0;  \r\nif (uv.x < 0.0 || uv.y < 0.0 || uv.x >= 1.0 || uv.y >= 1.0) return 0.0;\r\n\r\nuint2 g = FONT[code - 32u];\r\nuint2 p = (uint2)(uv * 8.0);\r\nuint rowBits = ((p.y < 4u) ? (g.x >> (p.y * 8u))\r\n                           : (g.y >> ((p.y - 4u) * 8u))) && 0xFFu;\r\nreturn (float)((rowBits >> p.x) && 1u);"
   outputType=CMOT_Float1
   inputs [
    {inputName=Time, input {expression=&self:MaterialExpressionMultiply_1}}
    {inputName=InUV, input {expression=&self:MaterialExpressionSaturate_1}}
    {
     inputName=Seed
     input {expression=&self:MaterialExpressionParticleRandom_0}
    }
    {inputName=Scale, input {expression=&self:MaterialExpressionConstant_0}}
   ]
  }
  n10 = Add(A=@n4, B=@n9)
  n11 = MaterialFunctionCall(Distance=ScalarParameter() {parameterName=Near Camera Fade}) {materialFunction=&Near_Camera_Fade}
  Multiply(A=Multiply(A=@n5.A, B=Min(A=@n10)), B=Multiply(A=@n11.Result, B=@n11.Result))
 ]
}
