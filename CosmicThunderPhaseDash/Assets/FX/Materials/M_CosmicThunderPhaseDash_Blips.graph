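# Translucent unlit Niagara mesh-particle material for the blips effect of Cosmic Thunder Phase Dash.
# A blip texture with tiled, clamped UVs is tinted by particle color and passed through shared emissive adjustments.
# Particle alpha times Fresnel feeds a depth fade for soft opacity.
# Panning noise sampled in particle-relative world space adds to a UV-weighted Z stretch, forming a vector that appears to displace the mesh.
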Material /CosmicThunderPhaseDash/Assets/FX/Materials/M_CosmicThunderPhaseDash_Blips
graph {
 properties {
  bAutomaticallySetUsageInEditor=false
  bUsedWithNiagaraMeshParticles=true
  blendMode=BLEND_Translucent
  shadingModel=MSM_Unlit
  usedShadingModels=Unlit
 }
 nodes [
  n1 = WorldPosition()
  n2 = ParticlePositionWS()
  n3 = Add(
   A=Multiply(A=Subtract(A=@n1.XYZ, B=@n2), B=ScalarParameter() {defaultValue=0.07, parameterName=Noise Scale})
   B=Multiply(A=Time(), B=VectorParameter() {parameterName=Noise Pan Speed, defaultValue=<rgba 0 0 -0.5 0>}.RGB)
  )
  n4 = TextureCoordinate()
  n5 = Power(Base=OneMinus(None=ComponentMask(None=@n4) {g=true}), Exp=ScalarParameter() {defaultValue=3, parameterName=Strtch Pow})
  n6 = Constant3Vector() {constant=<rgba 0 0 1 0>}
  Add(
   A=Multiply(
    A=TextureSample(UVs=@n3) {texture=/Game/Effects/Explosion_Update/Textures/VT_ExplosionNoise}.RGB
    B=ScalarParameter() {defaultValue=2, parameterName=Noise Strength}
   )
   B=Multiply(A=Multiply(A=@n5, B=@n6), B=ScalarParameter() {defaultValue=5, parameterName=Stretch})
  )
  n7 = ParticleColor()
  n8 = ScalarParameter() {defaultValue=1.1, parameterName=Fresnel Brightness}
  n9 = ScalarParameter() {defaultValue=1, parameterName=Fresnel Expo}
  DepthFade(
   Opacity=Multiply(
    A=@n7.A
    B=MaterialFunctionCall(Brightness=@n8, Exponent=@n9) {materialFunction=&MF_Fresnel_FX}.Fresnel
   )
   FadeDistance=ScalarParameter() {defaultValue=30, parameterName=Depth Fade}
  )
  n10 = Saturate(None=Add(A=TextureCoordinate() {uTiling=3, vTiling=2}, B=Constant2Vector() {g=-0.5}))
  MaterialFunctionCall(Emissive=Multiply(A=TextureSample(UVs=@n10) {texture=&/T_CosmicThunderPhaseDash_Blips}.RGB, B=@n7.RGB)) {materialFunction=&MF_FX_EmissiveAdjustments}
 ]
}
