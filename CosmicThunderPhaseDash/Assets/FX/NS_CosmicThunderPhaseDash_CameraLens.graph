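# Niagara system for the camera-lens effect of the Cosmic Thunder Phase Dash.
# It combines six emitters: two sphere distortions, a screen wave, two wide glitch emitters and an ASCII emitter.
# User parameters expose a lifetime scale and an Overload value.
# System scripts use SystemState, a set-variables assignment, multiply inputs and a platform-dependent float pick, suggesting per-platform tuning.
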NiagaraSystem /CosmicThunderPhaseDash/Assets/FX/NS_CosmicThunderPhaseDash_CameraLens
dependencies {
 usedDataInterfaces [&NiagaraDataInterfacePlatformSet]
 usedDynamicInputs [
  &Multiply_Float
  &Multiply_LinearColorByFloat
  &FN_PickFloatDependOnPlatform
 ]
 usedModules [
  &self:SystemSpawnScript.SystemScriptSource.SystemScriptGraph.NiagaraNodeAssignment_2.SetVariables_#1
  &SystemState
 ]
}
emitters {
 ASCII {rendererClasses [], renderers []}
 GlitchesWide {rendererClasses [], renderers []}
 GlitchesWide001 {rendererClasses [], renderers []}
 Screen_Wave {rendererClasses [], renderers []}
 Sphere_Distortion {rendererClasses [], renderers []}
 Sphere_Distortion001 {rendererClasses [], renderers []}
}
summary {
 systemName=NS_CosmicThunderPhaseDash_CameraLens
 emitters [
  {emitterName=Sphere_Distortion, rendererClasses []}
  {emitterName=Sphere_Distortion001, rendererClasses []}
  {emitterName=Screen_Wave, rendererClasses []}
  {emitterName=GlitchesWide, rendererClasses []}
  {emitterName=GlitchesWide001, rendererClasses []}
  {emitterName=ASCII, rendererClasses []}
 ]
 userVariables [
  User.LifetimeScale {type=NiagaraFloat, defaultValue=1, description=""}
  User.Overload {type=NiagaraFloat, defaultValue=0, description=""}
 ]
}
