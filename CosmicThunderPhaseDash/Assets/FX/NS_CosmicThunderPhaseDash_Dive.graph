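# Niagara effect for the dive of the Cosmic Thunder Phase Dash, wrapping a character in short-lived purple, magenta and red energy.
# GPU ember sprites spawn from the skeletal mesh, while box-spawned glitch meshes flicker through color curves.
# Line meshes follow eleven body bones alongside a light renderer.
# Spline_Ribbons uses bone-attached meshes rather than ribbon renderers, facing against smoothed system velocity, with length and material driven by normalized velocity.
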NiagaraSystem /CosmicThunderPhaseDash/Assets/FX/NS_CosmicThunderPhaseDash_Dive
dependencies {
 usedDataInterfaces [
  &NiagaraDataInterfaceArrayColor
  &NiagaraDataInterfaceArrayFloat3
  &NiagaraDataInterfaceCurve
  &NiagaraDataInterfaceColorCurve
  &NiagaraDataInterfaceSkeletalMesh
 ]
 usedDynamicInputs [&RemapRange, &VectorLength]
 usedModules [
  &self:SystemSpawnScript.SystemScriptSource.SystemScriptGraph.NiagaraNodeAssignment_0.SetVariables_#1
  &SystemState
  &InterpolateOverTime
  &self:SystemSpawnScript.SystemScriptSource.SystemScriptGraph.NiagaraNodeAssignment_1.SetVariables_#2
  &self:SystemSpawnScript.SystemScriptSource.SystemScriptGraph.NiagaraNodeAssignment_2.SetVariables_#3
 ]
}
emitters {
 Embers {
  simTarget=GPUComputeSim
  emitterUpdateScript [
   &EmitterState(Life Cycle Mode=::NewEnumerator1, Loop Behavior=::NewEnumerator1, Loop Duration=0.4, UseLoopDelay=true)
   &SpawnBurst_Instantaneous(Spawn Count=15, Spawn Time=0.02, Use Spawn Probability=false, UseLoopCountLimit=false) {module=SpawnBurst_Instantaneous001}
   &SpawnBurst_Instantaneous(Spawn Count=20, Use Spawn Probability=false, UseLoopCountLimit=false)
   &SpawnRate(SpawnRate=100, Use Spawn Probability=false)
  ]
  particleSpawnScript [
   /SetVariables(Particles.SkeletonSamplingIndex:NiagaraInt32=/CascadeToNiagaraConverter/NiagaraScripts/DynamicInputScripts/SequentialInt(Use Array Data:NiagaraBool=false, First Number:NiagaraInt32=0, Sequence:NiagaraDataInterfaceArrayInt32=?, Last Number:NiagaraInt32=13)) {isSetParameters=true}
   &InitializeParticle(
    Color=&SelectLinearColorFromArray(
     Array Sampling Mode=::NewEnumerator0
     Direct Array Index=?
     Color Selection Array=DataInterface() {propertyValues {ColorData [<rgba 1 0 0.0381 1>, <rgba 0.0496 0 0.2865 1>, <rgba 0.2865 0 0.2286 1>]}}
     Random Seed=0
    )
    Lifetime=&RandomRangeFloat(Minimum=0.2, Maximum=0.6, Evaluation Type=::NewEnumerator1)
    Sprite Size=&Vector2DFromFloat(Value=&RandomRangeFloat(Minimum=3, Maximum=5, Evaluation Type=::NewEnumerator1))
    Sprite Size Mode=::NewEnumerator4
    UsePositionOffset=false
    Write Material Random=true
   )
   &SkeletalMeshLocation(
    MaskSpawnGroup=false
    Orientation Sampling=::NewEnumerator1
    OverrideRandomSeed=false
    Sampled Position Offset=&SelectVectorFromArray(
     Direct Array Index=$Particles.SkeletonSamplingIndex
     Vector Selection Array=DataInterface() {propertyValues {FloatData [<0 0 0>, <5 0 0>, <0 0 0>, <0 0 0>, <5 0 0>, <0 0 0>, <0 0 0>, <-5 0 0>, <-10 0 0>, <8 0 0>, <-8 0 0>, <-4 0 0>, <8 0 0>, <4 0 0>]}}
    )
    Skeletal Mesh=$System.Skeletal Mesh Reader
   )
   &InheritVelocity(Inherited Velocity Amount Scale=<-0.2 -0.2 -0.2>, Inherited Velocity Speed Limit=0, Speed Limit=false, UseSpeedThreshold=false)
   &AddVelocity(Velocity=&RandomRangeVector(Minimum=<-40 -40 -20>, Maximum=<40 40 20>, Evaluation Type=::NewEnumerator1))
   &CameraOffset(Camera Offset Amount=30)
  ]
  particleUpdateScript [
   &ParticleState()
   &DynamicMaterialParameters(
    Index 0 Param 1=&RandomRangeFloat(Minimum=0.7, Maximum=3)
    Index 0 Param 2=&RandomRangeFloat(Minimum=0.01, Maximum=0.1, Evaluation Type=::NewEnumerator1)
    Index 0 Param 3=&RandomRangeFloat(Minimum=-0.05, Maximum=0)
    Index 0 Param 4=&FloatFromCurve(
     FloatCurve=DataInterface() {propertyValues {Curve {
        defaultValue=1
        keys [
         {arriveTangent=9.0725, interpMode=RCIM_Cubic, leaveTangent=9.0725, tangentMode=RCTM_User, time=0.0252, value=0.9989}
         {arriveTangent=-2.5249, interpMode=RCIM_Cubic, leaveTangent=-2.5249, tangentMode=RCTM_User, time=0.1559, value=0.0555}
         {arriveTangent=1.3524, interpMode=RCIM_Cubic, leaveTangent=1.3524, time=0.6844, value=0.1079}
         {arriveTangent=-0.0405, interpMode=RCIM_Cubic, leaveTangent=-0.0405, tangentMode=RCTM_User, time=1.0252, value=1.2312}
        ]
       }}}
     Scale Curve=10
    )
    Module.Index0Float0Write=true
    Module.Index0Float1Write=true
    Module.Index0Float2Write=true
    Module.Index0Float3Write=true
   )
   &ScaleSpriteSize(
    Scale Sprite Size Mode=::NewEnumerator0
    Uniform Scale Factor=&FloatFromCurve(FloatCurve=DataInterface() {propertyValues {Curve {keys [
         <at -0.0044 0.2654>
         {arriveTangent=1.5329, interpMode=RCIM_Cubic, leaveTangent=1.5329, time=0.1047, value=0.9233}
         {arriveTangent=-0.1211, interpMode=RCIM_Cubic, leaveTangent=-0.1211, time=0.4609, value=0.9788}
         {arriveTangent=-1.5371, interpMode=RCIM_Cubic, leaveTangent=-1.5371, time=0.8919, value=0.828}
         <at 0.9974 0.1543>
        ]}}})
   )
   &ScaleColor(Scale Alpha=2, ScaleA=true, ScaleRGB=true)
   &Drag(
    Drag=&RandomRangeFloat(Minimum=5, Maximum=8, Evaluation Type=::NewEnumerator1)
    Use Linear Drag=true
    Use Rotational Drag=true
   )
   &AccelerationForce(Acceleration=<0 0 -100>) {enabled=false}
   &SolveForcesAndVelocity(Clamp Velocity=false, Limit Acceleration=false)
  ]
  rendererClasses [&NiagaraSpriteRendererProperties]
  renderers [{rendererClass=&NiagaraSpriteRendererProperties}]
 }
 Glitch_sprite {
  emitterUpdateScript [
   &EmitterState(Life Cycle Mode=::NewEnumerator1, Loop Behavior=::NewEnumerator1, Loop Duration=0.7, UseLoopDelay=false)
   &SpawnRate(SpawnRate=&Multiply_Float(A=70, B=1), Use Spawn Probability=false)
  ]
  particleSpawnScript [
   &InitializeParticle(
    Color=&MultiplyLinearColor(
     Linear Color=&MakeCustomLinearColorFromBool(
      Bool=&RandomBool()
      True Color=&MakeCustomLinearColorFromBool(Bool=&RandomBool(), True Color=<rgba 0.4257 0.0762 0.4531 2>, False Color=<rgba 0.0484 0.0216 0.2396 1>)
      False Color=<rgba 0.1551 0.0522 0.5 1>
     )
     Scale Factor=<1 1 1 1>
    )
    Lifetime Max=0.3
    Lifetime Min=0.3
    Lifetime Mode=::NewEnumerator1
    Mesh Scale Max=&MakeCustomVectorFromBool(Bool=&RandomBool(), True Vector=<1 6 1>, False Vector=<1 1 8>)
    Mesh Scale Min=<1 1 3>
    Mesh Scale Mode=::NewEnumerator2
    Position Offset=<0 0 100>
    UsePositionOffset=true
    Write Material Random=true
   )
   &CameraOffset(Camera Offset Amount=50)
   &ShapeLocation(
    Box Size=&FN_FrontEndVector3Modifier(Gameplay=<62 62 100>, Front End=<80 80 100>)
    MaskSpawnGroup=false
    Override Seed=false
    Shape Primitive=::NewEnumerator2
    Surface Only=false
    UseOwnerScale=true
   )
  ]
  particleUpdateScript [
   &ParticleState()
   &ScaleMeshSize(Scale Factor=&Multiply_VectorByFloat(
     Vector=<1 1 1>
     Float=&Multiply_Float(
      A=0.2
      B=&FloatFromCurve(
       FloatCurve=DataInterface() {propertyValues {Curve {keys [
           {arriveTangent=4.1785, time=0, value=1}
           <at 0.5 1.1>
           {arriveTangent=-0.0405, interpMode=RCIM_Cubic, leaveTangent=-0.0405, tangentMode=RCTM_User, time=1, value=1.5}
          ]}}}
       Scale Curve=1.5
      )
     )
    ))
   &Drag(Drag=2, Use Linear Drag=true, Use Rotational Drag=true)
   &SolveForcesAndVelocity(Clamp Velocity=false, Limit Acceleration=false)
   &DynamicMaterialParameters(Module.Param0WriteEnabled=true, Module.Param1WriteEnabled=true, Module.Param2WriteEnabled=true, Module.Param3WriteEnabled=true)
   &ScaleColor(
    Scale Alpha=&Multiply_Float(
     A=&FloatFromCurve(
      FloatCurve=DataInterface() {propertyValues {Curve {keys [
          {arriveTangent=9.0725, interpMode=RCIM_Cubic, leaveTangent=9.0725, tangentMode=RCTM_User, time=0.0046, value=0.1672}
          {arriveTangent=-2.5249, interpMode=RCIM_Cubic, leaveTangent=-2.5249, tangentMode=RCTM_User, time=0.1582, value=0.9847}
          {arriveTangent=-1.1697, interpMode=RCIM_Cubic, leaveTangent=-1.1697, time=0.5512, value=0.7526}
          {arriveTangent=-0.0405, interpMode=RCIM_Cubic, leaveTangent=-0.0405, tangentMode=RCTM_User, time=1, value=0}
         ]}}}
      Scale Curve=&RandomRangeFloat(Minimum=3, Maximum=5)
     )
     B=0.5
    )
    Scale RGB=&Multiply_VectorByFloat(
     Vector=&MakeVectorFromLinearColorRGB(Color=&ColorFromCurve(ColorCurve=DataInterface() {propertyValues {
         AlphaCurve {keys []}
         BlueCurve {keys [
           <at 0 4.2929 RCIM_Cubic>
           <at 0.1326 5>
           <at 0.2869 5>
           <at 0.4134 0>
           <at 0.5108 5>
           <at 0.6236 0.8907>
           {arriveTangent=22.0813, interpMode=RCIM_Cubic, leaveTangent=22.0813, time=0.7107, value=1}
           <at 0.8097 5>
          ]}
         GreenCurve {keys [
           <at 0 2.1716 RCIM_Cubic>
           <at 0.1326 0>
           <at 0.2869 0.6611>
           <at 0.4134 0.0442>
           <at 0.5108 0.7585>
           <at 0.6236 0>
           {arriveTangent=3.5376, interpMode=RCIM_Cubic, leaveTangent=3.5376, time=0.7107, value=0.251}
           <at 0.8097 0.6583>
          ]}
         RedCurve {keys [
           <at 0 5 RCIM_Cubic>
           <at 0.1326 2.465>
           <at 0.2869 0>
           <at 0.4134 5>
           <at 0.5108 4.6237>
           <at 0.6236 5>
           {arriveTangent=-8.5763, interpMode=RCIM_Cubic, leaveTangent=-8.5763, time=0.7107, value=0.2471}
           <at 0.8097 3.404>
          ]}
        }}))
     Float=&FloatFromCurve(
      FloatCurve=DataInterface() {propertyValues {Curve {keys [{arriveTangent=-2.0684, interpMode=RCIM_Cubic, leaveTangent=-2.0684, tangentMode=RCTM_User, time=0, value=1}, <at 1 0>]}}}
      Scale Curve=&Multiply_Float(A=1, B=1)
     )
    )
    ScaleA=true
    ScaleRGB=true
   )
  ]
  rendererClasses [&NiagaraMeshRendererProperties]
  renderers [{rendererClass=&NiagaraMeshRendererProperties, rendererIndex=0}]
 }
 Line_rate {
  emitterUpdateScript [
   &EmitterState(Life Cycle Mode=::NewEnumerator1, Loop Behavior=::NewEnumerator1, Loop Delay=0.8, Loop Duration=0.5, UseLoopDelay=false)
   &SpawnRate(
    SpawnRate=&FloatFromCurve(
     FloatCurve=DataInterface() {propertyValues {Curve {keys [
         {interpMode=RCIM_Cubic, tangentMode=RCTM_User, time=0, value=200}
         {arriveTangent=-493.5296, interpMode=RCIM_Cubic, leaveTangent=-493.5296, tangentMode=RCTM_User, time=0.3596, value=100.2756}
         <at 1 0>
        ]}}}
     CurveIndex=$Emitter.NormalizedLoopAge
    )
    Use Spawn Probability=false
   )
   /SetVariables(
    Emitter.Color:LinearColor=&ColorFromCurve(
     ColorCurve=DataInterface() {propertyValues {
       AlphaCurve {keys [<at 0 1 RCIM_Cubic>,<at 1 0 RCIM_Cubic>]}
       BlueCurve {keys [<at 0.2824 0.01 RCIM_Cubic>,<at 0.539 3.4136>]}
       GreenCurve {keys [<at 0.2824 0 RCIM_Cubic>,<at 0.539 0>]}
       RedCurve {keys [<at 0.2824 5 RCIM_Cubic>,<at 0.539 0.634>]}
      }}
     CurveIndex=$Emitter.NormalizedLoopAge
    )
    Emitter.LitPosition:NiagaraPosition=<0 0 65>
    Emitter.LitRadius=140
   ) {isSetParameters=true}
  ]
  particleSpawnScript [
   &FN_SetSpawnGroup(Number of Spawn Groups=11)
   &InitializeParticle(
    Color=&MultiplyLinearColor(
     Linear Color=&MakeCustomLinearColorFromBool(
      Bool=&RandomBool()
      True Color=&MakeCustomLinearColorFromBool(Bool=&RandomBool(), True Color=<rgba 0.849 0 0.3384 1>, False Color=<rgba 0.0151 0 0.1198 1>)
      False Color=<rgba 0.2157 0 0.5 1>
     )
     Scale Factor=<1 1 1 1>
    )
    Lifetime Max=0.3
    Lifetime Min=0.2
    Lifetime Mode=::NewEnumerator1
    Mesh Scale Max=<0.7 0.7 25>
    Mesh Scale Min=<0.3 0.3 10>
    Mesh Scale Mode=::NewEnumerator2
    Position Offset=<0 0 100>
    UsePositionOffset=true
    Write Material Random=true
   )
   &ShapeLocation(
    Apply To Particle Position=::NewEnumerator1
    Box Size=<20 20 20>
    MaskSpawnGroup=false
    Override Seed=false
    Shape Origin=&VectorFromFloat(Value=0)
    Shape Primitive=::NewEnumerator2
    Surface Only=false
    UseOwnerScale=true
   )
   &CameraOffset(Camera Offset Amount=50)
   &AddVelocity(Velocity=<0 0 200>)
  ]
  particleUpdateScript [
   &ParticleState()
   &SkeletalMeshLocation(
    Bone / Socket Index=$Particles.SpawnGroup
    Bone Sampling Mode=::NewEnumerator4
    MaskSpawnGroup=false
    Orientation Sampling=::NewEnumerator1
    Sampled Position Offset=$Particles.ShapeLocation.ShapePosition
    Skeletal Mesh=DataInterface() {propertyValues {
      DefaultMesh=null
      ExcludeBoneName=root
      OptionalComponentClass=null
      PreviewMesh=&M_Med_Soldier_04_Preview
      SkinningMode=SkinOnTheFly
      SoftSourceActor=null
      SourceMode=Default
      UvSetIndex=0
      WholeMeshLOD=-1
      bExcludeBone=false
      bReadDeformedGeometry=true
      bRequireCurrentFrameData=true
      ComponentTags []
      FilteredBones [head,clavicle_l,clavicle_r,upperarm_l,upperarm_r,hand_l,hand_r,thigh_l,thigh_r,calf_l,calf_r]
      FilteredSockets []
      MeshUserParameter {parameter {name=None, typeDefHandle {registeredTypeIndex=1}, varData []}}
      SamplingRegions []
     }}
   )
   &ScaleMeshSize(Scale Factor=&Multiply_VectorByFloat(
     Vector=&MakeVector(
      X=1
      Y=1
      Z=&FloatFromCurve(FloatCurve=DataInterface() {propertyValues {Curve {keys [<at 0 1>,<at 0.7 1>,<at 1 0>]}}})
     )
     Float=&Multiply_Float(
      A=0.1
      B=&FloatFromCurve(
       FloatCurve=DataInterface() {propertyValues {Curve {keys [
           {arriveTangent=4.1785, time=0, value=0}
           <at 0.5 1.1>
           {arriveTangent=-0.0405, interpMode=RCIM_Cubic, leaveTangent=-0.0405, tangentMode=RCTM_User, time=1, value=1.5}
          ]}}}
       Scale Curve=&Multiply_Float(
        A=1
        B=&ClampFloat(
         Float=&FN_LOD_Distance_Remap(Start Distance=200, End Distance=400, Minimum Output=0.5, Maximum Output=1)
         Min=0.5
        )
       )
      )
     )
    ))
   &Drag(Drag=4, Use Linear Drag=true, Use Rotational Drag=true)
   &SolveForcesAndVelocity(Clamp Velocity=false, Limit Acceleration=false)
   &ScaleColor(
    Scale Alpha=&Multiply_Float(
     A=&FloatFromCurve(
      FloatCurve=DataInterface() {propertyValues {Curve {keys [<at 0 1>,<at 1 0>]}}}
      Scale Curve=&RandomRangeFloat(Minimum=0.4, Maximum=1)
     )
     B=3
    )
    Scale RGB=&Multiply_VectorByFloat(
     Vector=&MakeVectorFromLinearColorRGB(Color=&ColorFromCurve(ColorCurve=DataInterface() {propertyValues {
         AlphaCurve {keys []}
         BlueCurve {keys [<at 0 5>,<at 0.2711 0.599>,<at 0.4778 1.25>,<at 0.6236 5>,<at 0.7107 1 RCIM_Cubic>,<at 0.8097 5>]}
         GreenCurve {keys [
           <at 0 0>
           <at 0.2711 0>
           <at 0.4778 0>
           <at 0.6236 0.6526>
           {arriveTangent=1.8289, interpMode=RCIM_Cubic, leaveTangent=1.8289, time=0.7107, value=0}
           <at 0.8097 0.9929>
          ]}
         RedCurve {keys [<at 0 2.465>,<at 0.2711 0.0376>,<at 0.4778 0.853>,<at 0.6236 0>,<at 0.7107 0.2867 RCIM_Cubic>,<at 0.8097 0>]}
        }}))
     Float=&FloatFromCurve(
      FloatCurve=DataInterface() {propertyValues {Curve {keys [{arriveTangent=-2.0684, interpMode=RCIM_Cubic, leaveTangent=-2.0684, tangentMode=RCTM_User, time=0, value=2}, <at 1 0>]}}}
      Scale Curve=&Multiply_Float(A=&Multiply_Float(A=1.5, B=1), B=1)
     )
    )
    ScaleA=true
    ScaleRGB=true
   )
  ]
  rendererClasses [&NiagaraMeshRendererProperties, &NiagaraLightRendererProperties]
  renderers [
   {rendererClass=&NiagaraMeshRendererProperties, rendererIndex=0}
   {rendererClass=&NiagaraLightRendererProperties, rendererIndex=1}
  ]
 }
 Spline_Ribbons {
  emitterUpdateScript [
   &EmitterState(Inactive Response=::NewEnumerator1, Life Cycle Mode=::NewEnumerator1, Loop Behavior=::NewEnumerator1, Loop Duration=1, UseLoopDelay=false)
   &SpawnBurst_Instantaneous(Spawn Count=11, Use Spawn Probability=false, UseLoopCountLimit=true)
   /SetVariables(
    Emitter.Width1:NiagaraFloat=&FloatFromCurve(
     FloatCurve=DataInterface() {propertyValues {Curve {
        defaultValue=1
        keys [
         {arriveTangent=4.1785, interpMode=RCIM_Cubic, leaveTangent=4.1785, tangentMode=RCTM_User, time=0, value=0}
         {arriveTangent=1.1179, interpMode=RCIM_Cubic, leaveTangent=1.1179, tangentMode=RCTM_User, time=0.1397, value=0.6811}
         {arriveTangent=0.0224, interpMode=RCIM_Cubic, leaveTangent=0.0224, tangentMode=RCTM_User, time=0.6, value=1}
        ]
       }}}
     CurveIndex=$Emitter.NormalizedLoopAge
     Scale Curve=20
    )
    Emitter.Width2:NiagaraFloat=&FloatFromCurve(
     FloatCurve=DataInterface() {propertyValues {Curve {
        defaultValue=1
        keys [
         {arriveTangent=4.1785, interpMode=RCIM_Cubic, leaveTangent=4.1785, tangentMode=RCTM_User, time=0, value=0}
         {arriveTangent=1.1179, interpMode=RCIM_Cubic, leaveTangent=1.1179, tangentMode=RCTM_User, time=0.1397, value=0.6811}
         {arriveTangent=0.0224, interpMode=RCIM_Cubic, leaveTangent=0.0224, tangentMode=RCTM_User, time=0.6, value=1}
        ]
       }}}
     CurveIndex=$Emitter.NormalizedLoopAge
     Scale Curve=25
    )
   ) {isSetParameters=true}
  ]
  particleSpawnScript [
   &PartitionParticles(Number Of Partitions=11)
   &InitializeParticle(Color=<rgba 1 1 1 0.55>, Lifetime=0.7, Mesh Scale Mode=::NewEnumerator1, UsePositionOffset=false, Write Material Random=true)
   &CameraOffset(Camera Offset Amount=-10)
  ]
  particleUpdateScript [
   &ParticleState()
   &ScaleMeshSize(Scale Factor=&MakeVector(
     X=&FloatFromCurve(
      FloatCurve=DataInterface() {propertyValues {Curve {
         defaultValue=1
         keys [<at 0 0 RCIM_Cubic>, {arriveTangent=1.7579, interpMode=RCIM_Cubic, leaveTangent=1.7579, tangentMode=RCTM_User, time=1, value=1}]
        }}}
      CurveIndex=$System.NormalizedVelocityRange
      Scale Curve=0.9
     )
     Y=0.5
     Z=0.5
    ))
   &SkeletalMeshLocation(
    Bone / Socket Index=$Particles.PartitionParticles.Partition
    Bone Sampling Mode=::NewEnumerator4
    MaskSpawnGroup=false
    Orientation Sampling=::NewEnumerator1
    Sampled Orientation Coordinate Space=::World
    Skeletal Mesh=$System.Skeletal Mesh Reader
   )
   &UpdateMeshOrientation(
    Facing Coordinate Space:ENiagaraCoordinateSpace=::World
    Facing Direction=&Multiply_VectorByFloat(Vector=$System.SmoothedVelocity, Float=-1)
    Facing Interpolation=false
    Orientation Method=::NewEnumerator1
    Override Quat=false
    TransformFacing=true
    TransformUp=false
   )
   &DynamicMaterialParameters(
    Index 0 Param 1=&FloatFromCurve(
     FloatCurve=DataInterface() {propertyValues {Curve {
        defaultValue=1
        keys [<at 0 0 RCIM_Cubic>, {arriveTangent=1.7579, interpMode=RCIM_Cubic, leaveTangent=1.7579, tangentMode=RCTM_User, time=1, value=1}]
       }}}
     CurveIndex=$System.NormalizedVelocityRange
     Scale Curve=2
    )
    Index 0 Param 2=1
    Index 0 Param 3=&RandomRangeFloat(Minimum=1, Maximum=2.5)
    Module.Index0Float0Write=true
    Module.Index0Float1Write=false
    Module.Index0Float2Write=true
    Module.Index0Float3Write=true
   )
   &ScaleColor(
    Linear Color Curve=DataInterface() {propertyValues {
      AlphaCurve {keys [<at 0.0041 0>,<at 0.354 0.8061 RCIM_Cubic>,<at 0.4015 1>,<at 0.9 0 RCIM_Cubic>]}
      BlueCurve {keys [<at 0.2897 0.1099>,<at 1 0.0002>]}
      GreenCurve {keys [<at 0.2897 0>,<at 1 0>]}
      RedCurve {keys [<at 0.2897 1>,<at 1 0.0833>]}
     }}
    Scale Mode=::NewEnumerator2
   )
   /SetVariables(
    Particles.Color2=&MultiplyLinearColor(
     Linear Color=&MakeCustomLinearColorFromBool(
      Bool=&RandomBool()
      True Color=&MakeCustomLinearColorFromBool(Bool=&RandomBool(), True Color=<rgba 0.4257 0.0762 0.4531 2>, False Color=<rgba 0.0484 0.0216 0.2396 1>)
      False Color=<rgba 0.1551 0.0522 0.5 1>
     )
     Scale Factor=&MakeVector4(
      X=1
      Y=1
      Z=1
      W=&FloatFromCurve(
       FloatCurve=DataInterface() {propertyValues {Curve {
          defaultValue=1
          keys [
           {arriveTangent=9.0725, interpMode=RCIM_Cubic, leaveTangent=9.0725, tangentMode=RCTM_User, time=-0.0034, value=0.1728}
           {arriveTangent=-2.5249, interpMode=RCIM_Cubic, leaveTangent=-2.5249, tangentMode=RCTM_User, time=0.1408, value=0.9847}
           {arriveTangent=-1.146, interpMode=RCIM_Cubic, leaveTangent=-1.146, time=0.7222, value=0.6432}
           {arriveTangent=-0.0405, interpMode=RCIM_Cubic, leaveTangent=-0.0405, tangentMode=RCTM_User, time=1, value=0}
          ]
         }}}
       Scale Curve=3
      )
     )
    )
    Particles.MeshScale2:Vector3f=&Multiply_VectorByFloat(Vector=&RandomRangeVector(Minimum=<0.5 0.5 5>, Maximum=<2 2 2>), Float=0.7)
   ) {isSetParameters=true}
  ]
  rendererClasses [&NiagaraMeshRendererProperties]
  renderers [
   {rendererClass=&NiagaraMeshRendererProperties, rendererIndex=0}
   {rendererClass=&NiagaraMeshRendererProperties, rendererIndex=1}
  ]
 }
}
summary {
 systemName=NS_CosmicThunderPhaseDash_Dive
 emitters [
  {emitterName=Embers, simTarget=GPUComputeSim, rendererClasses [&NiagaraSpriteRendererProperties]}
  {emitterName=Spline_Ribbons, rendererClasses [&NiagaraMeshRendererProperties]}
  {emitterName=Line_rate, rendererClasses [&NiagaraMeshRendererProperties, &NiagaraLightRendererProperties]}
  {emitterName=Glitch_sprite, rendererClasses [&NiagaraMeshRendererProperties]}
 ]
 userVariables [
  User.bIsFrontEnd {type=NiagaraBool}
  User.bIsFrontEndPreview {type=NiagaraBool}
 ]
}
