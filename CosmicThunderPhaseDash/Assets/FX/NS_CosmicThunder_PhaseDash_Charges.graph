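# Niagara system that displays Cosmic Thunder Phase Dash charges as floating skeletal-mesh-relative meshes with sprites, hidden or faded while aiming down sights.
# Charges spawns one particle per maximum charge, adds noise, bobbing and velocity-scaled smoothing, and compares each partition to NumCharges for visibility and scale-in.
# A scratch module sends events on visibility changes, alongside ShineBurst and Skull_Bits burst emitters.
# Energy001 spawns mesh trails from Charges; the sprite-based Energy is disabled.
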NiagaraSystem /CosmicThunderPhaseDash/Assets/FX/NS_CosmicThunder_PhaseDash_Charges
dependencies {
 usedDataInterfaces [
  &NiagaraDataInterfaceCurve
  &NiagaraDataInterfaceArrayFloat3
  &NiagaraDataInterfaceArrayFloat
  &NiagaraDataInterfaceSpriteRendererInfo
  &NiagaraDataInterfaceParticleRead
  &NiagaraDataInterfaceColorCurve
  &NiagaraDataInterfaceArrayColor
  &NiagaraDataInterfaceVectorCurve
  &NiagaraDataInterfaceSkeletalMesh
 ]
 usedDynamicInputs [
  &ClampFloat
  &RemapRange
  &VectorLength
  &SetBoolByFloatComparison
 ]
 usedModules [
  &self:SystemSpawnScript.SystemScriptSource.SystemScriptGraph.NiagaraNodeAssignment_4.SetVariables_#1
  &self:SystemSpawnScript.SystemScriptSource.SystemScriptGraph.NiagaraNodeAssignment_6.SetVariables_#2
  &self:SystemSpawnScript.SystemScriptSource.SystemScriptGraph.NiagaraNodeAssignment_2.SetVariables_#3
  &self:SystemSpawnScript.SystemScriptSource.SystemScriptGraph.NiagaraNodeAssignment_7.SetVariables_#4
  &self:SystemSpawnScript.SystemScriptSource.SystemScriptGraph.NiagaraNodeAssignment_9.SetVariables_#5
  &SystemState
  &self:SystemSpawnScript.SystemScriptSource.SystemScriptGraph.NiagaraNodeAssignment_1.SetVariables_#6
  &self:SystemSpawnScript.SystemScriptSource.SystemScriptGraph.NiagaraNodeAssignment_0.SetVariables_#7
  &self:SystemSpawnScript.SystemScriptSource.SystemScriptGraph.NiagaraNodeAssignment_8.SetVariables_#8
  &self:SystemSpawnScript.SystemScriptSource.SystemScriptGraph.NiagaraNodeAssignment_3.SetVariables_#9
 ]
}
emitters {
 Charges {
  emitterUpdateScript [
   &EmitterState(Life Cycle Mode=::NewEnumerator1, Loop Behavior=::NewEnumerator1, Loop Duration Mode=::NewEnumerator1, UseLoopDelay=false)
   &SpawnBurst_Instantaneous(Spawn Count=$User.MaxNumCharges, Use Spawn Probability=false, UseLoopCountLimit=false)
  ]
  particleSpawnScript [
   &InitializeParticle(
    Lifetime=2
    Mesh Scale Mode=::NewEnumerator1
    Mesh Uniform Scale Max=0.08
    Mesh Uniform Scale Min=0.07
    Sprite Rotation Mode=::NewEnumerator2
    Sprite Size=<28 20>
    Sprite Size Mode=::NewEnumerator4
    Sprite UV Mode=::NewEnumerator3
    UsePositionOffset=false
    Write Material Random=true
   )
   &PartitionParticles(Number Of Partitions=$User.MaxNumCharges)
   /SetVariables(Particles.Position=&TransformPosition(
     Position=&SelectVectorFromArray(
      Direct Array Index=$Particles.PartitionParticles.Partition
      Vector Selection Array=DataInterface() {propertyValues {FloatData [<-50 0 0>, <0 0 0>, <50 0 0>]}}
     )
     Source Space=::Local
     Destination Space=::Simulation
    )) {isSetParameters=true}
   &InitialMeshOrientation(Mesh Axis To Orient=<0 1 0>, Mesh Orientation Mode=::NewEnumerator2, Rotation=<0 0 -0.17>, Use Rotation Vector=true)
   &CameraOffset(Camera Offset Amount=&RandomRangeFloat(Minimum=25, Maximum=50)) {enabled=false}
   /SetVariables(Particles.DynamicParametersSprite:Vector4f=<25 0 0 0>, Particles.DynamicParametersSpriteDark:Vector4f=<0 1 0 0>) {isSetParameters=true}
   &ApplyInitialForces()
   /SetVariables(Particles.PreviousVisibility=1, Particles.glow camofs=-8) {isSetParameters=true}
  ]
  particleUpdateScript [
   &ParticleState(Kill Particles When Lifetime Has Elapsed=false, Let Infinitely Lived Particles Die When Emitter Deactivates=true)
   &SkeletalMeshLocation(
    Bone Sampling Mode=::NewEnumerator4
    "Interpolate Mesh Sampling (Update)"=true
    MaskSpawnGroup=false
    Orientation Sampling=::NewEnumerator1
    Position Sampling=::NewEnumerator2
    Sampled Position Offset:NiagaraPosition=$System.ChargeOffsets_Local
    Sampled Position Offset Coordinate Space:ENiagaraSkelMeshTransforms=::NewEnumerator2
    Skeletal Mesh=$System.Skeletal Mesh
   )
   &Timeline(
    Instantaneous Rewind=true
    Play=&SetBoolByIntComparison(A=$Particles.VisibilityTag, B=0, Comparison Type=::NewEnumerator4)
    Play Rate=4
   )
   &CurlNoiseForce(
    Noise Frequency=0.5
    Noise Quality / Cost=::NewEnumerator0
    Noise Strength=5
    Pan Noise=true
    Pan Noise Field=&VectorFromFloat(Value=0.1)
    Random Seed:NiagaraInt32=$Particles.PartitionParticles.Partition
    Randomization Vector:Vector3f=&SelectVectorFromArray(
     Direct Array Index=$Particles.PartitionParticles.Partition
     Vector Selection Array=DataInterface() {propertyValues {FloatData [<0.65 0.125 0.37>, <0.254 0.842 0.178>, <0.1353 0.524 0.777>]}}
    )
    Write to Intrinsic Parameters=false
   )
   /Niagara/Noise/Waveform(
    EnableExponent1=false
    Global Amplitude Scale=5
    Global Phase:NiagaraFloat=&Add_Float(A=$Emitter.Age, B=&RandomRangeFloat(Minimum=0, Maximum=6))
    GlobalClampMax=false
    GlobalClampMin=false
    UseInput1=false
    UsePhase1=false
    "[1] Amplitude Min/Max":Vector2f=<0 0.5>
    "[1] Frequency"=0.125
   ) {module=Waveform003}
   /SetVariables(Particles.Position=&AddVectorToPosition(
     Position=$Particles.SkeletalMeshLocation.SampledPosition
     Vector=&TransformVector(
      Vector=&AddVectorToPosition(
       Position=&SelectVectorFromArray(Direct Array Index=$Particles.PartitionParticles.Partition, Vector Selection Array=$System.ChargePositions)
       Vector=&Multiply_Vector(
        A=$Output.CurlNoiseForce.CurlNoiseForce
        B=&MakeVector(
         X=1
         Y=1
         Z=&Multiply_Float(
          A=$StackContext.Waveform003.Waveform
          B=&Multiply_Float(
           A=&SelectFloatFromArray(
            Direct Array Index=$Particles.PartitionParticles.Partition
            Float Selection Array=DataInterface() {propertyValues {FloatData [1.233,1.425,1.1752]}}
           )
           B=&RandomRangeFloat(Minimum=0.95, Maximum=1.117)
          )
         )
        )
       )
      )
      Destination Space=::Simulation
     )
    )) {isSetParameters=true}
   &InterpolateOverTime(
    Rate Of Change:NiagaraFloat=&FloatFromCurve(
     FloatCurve=DataInterface() {propertyValues {Curve {keys [<at 0 4 RCIM_Cubic>, {arriveTangent=6.4288, interpMode=RCIM_Cubic, leaveTangent=6.4288, tangentMode=RCTM_User, time=1, value=6}]}}}
     CurveIndex=$System.NormalizedVelocityRange
     Scale Curve=&SelectFloatFromArray(Direct Array Index=$Particles.PartitionParticles.Partition, Float Selection Array=DataInterface() {propertyValues {FloatData [3.5,4,3.5]}})
    )
    Target Vector Value:NiagaraPosition=$Particles.Position
    Variable Type=::NewEnumerator2
   ) {module=InterpolateOverTime002}
   /SetVariables(Particles.Position=&AddVectorToPosition(
     Position=&Lerp_Vector(
      A=Linked() {linkedVariable {name=Particles.Position, type=NiagaraPosition}}
      B=$StackContext.InterpolateOverTime002.InterpolateOverTime_V3_.Moving Average
      Alpha=$System.SmoothedPositionWeight
     )
     Vector=&MakeVector(
      X=0
      Y=0
      Z=&Multiply_Float(
       A=$StackContext.Waveform003.Waveform
       B=&Multiply_Float(
        A=&SelectFloatFromArray(
         Direct Array Index=$Particles.PartitionParticles.Partition
         Float Selection Array=DataInterface() {propertyValues {FloatData [1.233,1.425,1.1752]}}
        )
        B=&RandomRangeFloat(Minimum=0.95, Maximum=1.117)
       )
      )
     )
    )) {isSetParameters=true}
   /SetVariables(
    Particles.Color_Sprite_Dark:LinearColor=&MultiplyLinearColor(
     Linear Color=<rgba 1.2349 0 2 0.4>
     Scale Factor=&MakeFloat4FromVectorAndFloat(
      "Vector (RGB)"=<1 1 1>
      "Float (Alpha)"=&FloatFromCurve(FloatCurve=DataInterface() {propertyValues {Curve {keys [
           {interpMode=RCIM_Cubic, tangentMode=RCTM_User, time=0, value=0.75}
           {interpMode=RCIM_Cubic, tangentMode=RCTM_User, time=0.5, value=1}
           {tangentMode=RCTM_User, time=1, value=1}
          ]}}})
     )
    )
    Particles.Sprite_Size_Dark:Vector2f=&Multiply_Vector2D_ByFloat(
     Float=&Multiply_Float(
      A=&FloatFromCurve(
       FloatCurve=DataInterface() {propertyValues {Curve {keys [
           {interpMode=RCIM_Cubic, tangentMode=RCTM_User, time=0, value=0}
           {arriveTangent=1.9474, interpMode=RCIM_Cubic, leaveTangent=1.9474, tangentMode=RCTM_User, time=1, value=1}
          ]}}}
       CurveIndex=$Engine.System.Age
       Scale Curve=8
      )
      B=&FloatFromCurve(
       FloatCurve=DataInterface() {propertyValues {Curve {keys [
           {interpMode=RCIM_Cubic, tangentMode=RCTM_User, time=0, value=0.75}
           {interpMode=RCIM_Cubic, tangentMode=RCTM_User, time=0.5, value=1}
           {tangentMode=RCTM_User, time=1, value=1}
          ]}}}
       Scale Curve=&MakeCustomFloatFromBool(Bool=&Boolean_NotOperation(Bool=$User.blsADS), True Float=1, False Float=0)
      )
     )
     Vector2D=<5 5>
    )
   ) {isSetParameters=true}
   /SetVariables(
    Particles.Color_Sprite_Light:LinearColor=<rgba 0.0469 0 0.0146 2>
    Particles.Sprite_Size_Light:Vector2f=&Multiply_Vector2D_ByFloat(
     Float=&Multiply_Float(
      A=&FloatFromCurve(
       FloatCurve=DataInterface() {propertyValues {Curve {keys [
           {interpMode=RCIM_Cubic, tangentMode=RCTM_User, time=0, value=0}
           {arriveTangent=1.9474, interpMode=RCIM_Cubic, leaveTangent=1.9474, tangentMode=RCTM_User, time=1, value=1}
          ]}}}
       CurveIndex=$Engine.System.Age
       Scale Curve=&MakeCustomFloatFromBool(Bool=&Boolean_NotOperation(Bool=$User.blsADS), True Float=1, False Float=0)
      )
      B=&FloatFromCurve(FloatCurve=DataInterface() {propertyValues {Curve {keys [
           {interpMode=RCIM_Cubic, tangentMode=RCTM_User, time=0, value=0.75}
           {interpMode=RCIM_Cubic, tangentMode=RCTM_User, time=0.5, value=1}
           {tangentMode=RCTM_User, time=1, value=1}
          ]}}})
     )
     Vector2D=<18 18>
    )
   ) {isSetParameters=true}
   &ScaleMeshSize(Scale Factor=&VectorFromFloat(Value=&FloatFromCurve(
      FloatCurve=DataInterface() {propertyValues {Curve {keys [
          <at 0 0 RCIM_Cubic>
          {arriveTangent=1.5312, interpMode=RCIM_Cubic, leaveTangent=1.5312, time=0.4, value=0.15}
          {arriveTangent=1.4167, interpMode=RCIM_Cubic, leaveTangent=1.4167, time=0.8, value=1.225}
          <at 1 1 RCIM_Cubic>
         ]}}}
      CurveIndex=$StackContext.Timeline.NormalizedTime
     )))
   /SetVariables(Particles.VisibilityTag=&MakeCustomIntFromBool(
     Bool=&SetBoolByIntComparison(A=$Particles.PartitionParticles.Partition, B=$User.NumCharges, Comparison Type=::NewEnumerator2)
     True INT=1
     False INT=0
    )) {isSetParameters=true}
   &InterpolateOverTime(
    Rate Of Change:NiagaraFloat=&SelectFloatFromArray(Direct Array Index=$Particles.PartitionParticles.Partition, Float Selection Array=DataInterface() {propertyValues {FloatData [0.1,0,0.05]}})
    Target Vector Value:Vector3f=&NormalizeVector(
     Vector To Normalize=&Subtract_Position(
      A=&TransformPosition(Position=$Particles.Position, Source Space=::Local, Destination Space=::Local)
      B=&TransformPosition(Position=$Particles.Previous.Position, Source Space=::Local, Destination Space=::Local)
     )
     Safe Normalize=true
     Safe Fallback Threshold=0
     Safe Fallback Vector=<0.0001 0.0001 0.0001>
    )
    Variable Type=::NewEnumerator2
   )
   /Niagara/Modules/Update/Orientation/V2/OrientMeshToVector(
    Look At Direction:Vector3f=&NormalizeVector(
     Vector To Normalize=$StackContext.InterpolateOverTime.InterpolateOverTime_V3_.Moving Average
     Safe Normalize=true
     Safe Fallback Threshold=0
     Safe Fallback Vector=<0 0 0.001>
    )
    Override Quat=false
    Up Axis Direction:Vector3f=<0 0 1>
    UseOptionalSideAxis=false
    UseOptionalUpAxis=true
   ) {enabled=false}
   &DynamicMaterialParameters(
    Index 0 Param 1=1
    Index 1 Param 1=&RandomRangeFloat(Minimum=0.2, Maximum=0.2)
    Index 3 Param 1=&MakeFloatFromInt(Int=$Particles.VisibilityTag)
    Index 3 Param 2=&MakeFloatFromVector(Vector=$Particles.InterpolateOverTime.InterpolateOverTime_V3_.Moving Average, Channel=::NewEnumerator0)
    Index 3 Param 3=&MakeFloatFromVector(Vector=$Particles.InterpolateOverTime.InterpolateOverTime_V3_.Moving Average, Channel=::NewEnumerator1)
    Index 3 Param 4=&MakeFloatFromVector(Vector=$Particles.InterpolateOverTime.InterpolateOverTime_V3_.Moving Average, Channel=::NewEnumerator2)
    Module.Index0Param0Write=true
    Module.Index0Param1Write=true
    Module.Index0Param2Write=true
    Module.Index0Param3Write=true
    Module.Index1Param0Write=true
    Module.Index1Param1Write=true
    Module.Index1Param2Write=true
    Module.Index1Param3Write=true
    Module.Index2Param0Write=true
    Module.Index2Param1Write=true
    Module.Index2Param2Write=true
    Module.Index2Param3Write=false
    Module.Index3Param0Write=true
    Module.Index3Param1Write=true
    Module.Index3Param2Write=true
    Module.Index3Param3Write=true
    Write Parameter Index 1=true
    Write Parameter Index 2=true
    Write Parameter Index 3=true
   )
   &InterpolateOverTime(
    Rate Of Change:NiagaraFloat=&FloatFromCurve(
     FloatCurve=DataInterface() {propertyValues {Curve {keys [
         {arriveTangent=2.7275, interpMode=RCIM_Cubic, leaveTangent=2.7275, tangentMode=RCTM_User, time=0, value=1.5}
         {arriveTangent=-0.3153, interpMode=RCIM_Cubic, leaveTangent=-0.3153, tangentMode=RCTM_User, time=1, value=2.5}
        ]}}}
     CurveIndex=$System.NormalizedVelocityRange
     Scale Curve=&SelectFloatFromArray(Direct Array Index=$Particles.PartitionParticles.Partition, Float Selection Array=DataInterface() {propertyValues {FloatData [3,4,3.5]}})
    )
    Target Quat Value:Quat4f=$Particles.MeshOrientation
    Variable Type=::NewEnumerator5
   ) {module=InterpolateOverTime001}
   /SetVariables(Particles.SmoothedOrientation:Quat4f=$StackContext.InterpolateOverTime001.InterpolateOverTime_Quaternion.Moving Average) {isSetParameters=true}
   /SetVariables(Particles.SmoothedOrientation:Quat4f=&MultiplyQuaternion(
     Quaternion A=$StackContext.SmoothedOrientation
     Quaternion B=&MultiplyQuaternion(
      Quaternion A=&MakeQuatFromAxisAngle(
       Axis=<0 0 1>
       AngleInDegrees=/Niagara/DynamicInputs/Helpers/Waveform(
        Global Phase=&Multiply_Float(
         A=&Add_Float(
          A=&Add_Float(A=$Engine.System.Age, B=&RandomRangeFloat(Minimum=0, Maximum=10))
          B=&RandomRangeFloat(Minimum=0, Maximum=4)
         )
         B=0.1
        )
        Global Amplitude Scale=50
        "Waveform [1]"=::NewEnumerator0
        "[1] Frequency"=1
        "[1] Amplitude Min/Max"=<-0.1 1.5>
       )
      )
      Quaternion B=&MakeQuatFromAxisAngle(
       Axis=<1 0 0>
       AngleInDegrees=&Waveform(
        Global Phase=&Multiply_Float(
         A=&Add_Float(A=$System.Age, B=&RandomRangeFloat(Minimum=0, Maximum=5))
         B=0.5
        )
        Global Amplitude Scale=20
        "Waveform [1]"=::NewEnumerator12
        "[1] Frequency"=1
        "[1] Amplitude Min/Max"=<-1 1>
       )
      )
     )
    )) {isSetParameters=true}
   &self:ScratchModule(
    ConditionToSendEvent:NiagaraBool=&SetBoolByIntComparison(A=$Particles.VisibilityTag, B=$Particles.PreviousVisibility, Comparison Type=::NewEnumerator5)
    Use Event Probability=false
   ) {module=ScratchModule}
   /SetVariables(Particles.PreviousVisibility:NiagaraInt32=$Particles.VisibilityTag) {isSetParameters=true}
   &SolveForcesAndVelocity(Clamp Velocity=false, Limit Acceleration=false)
   &ScaleColor(
    Scale Alpha=&MakeCustomFloatFromBool(Bool=&Boolean_NotOperation(Bool=$User.blsADS), True Float=1, False Float=0)
    ScaleA=true
    ScaleRGB=false
   )
   &ScaleColor(
    Scale Alpha=&FN_Select_Float_from_Time_Of_Day(Morning=0.7, Day=0.7, Evening=0.7, Night=0.7)
    Scale RGB=&FN_Select_Vector_from_Time_Of_Day(
     Morning=&MakeVectorFromLinearColorRGB(Color=<rgba 0 1 1.5 2>)
     Day=&MakeVectorFromLinearColorRGB(Color=<rgba 0 0.625 1.5 2>)
     Evening=&MakeVectorFromLinearColorRGB(Color=<rgba 0 0.625 1.5 2>)
     Night=&MakeVectorFromLinearColorRGB(Color=<rgba 0 1.3333 2 2>)
    )
    ScaleA=true
    ScaleRGB=true
   ) {enabled=false, module=ScaleColor001}
   &SubUVAnimation(
    Random Start Frame=true
    "Sprite Renderer "=DataInterface() {propertyValues {SpriteRenderer=&self:Charges_0.NiagaraSpriteRendererProperties_2}}
    UseEndFrame=false
    UseStartFrame=false
   )
  ]
  rendererClasses [&NiagaraMeshRendererProperties, &NiagaraSpriteRendererProperties]
  renderers [
   {rendererClass=&NiagaraMeshRendererProperties, rendererIndex=0}
   {rendererClass=&NiagaraSpriteRendererProperties, rendererIndex=1}
   {rendererClass=&NiagaraSpriteRendererProperties, rendererIndex=2}
  ]
 }
 Energy {
  bEnabled=false
  emitterUpdateScript [
   &EmitterState(Inactive Response=::NewEnumerator1, Life Cycle Mode=::NewEnumerator1, Loop Duration=1, UseLoopDelay=false)
   &SpawnParticlesFromOtherEmitter(
    Attribute Reader=DataInterface() {propertyValues {EmitterBinding {bindingMode=Other, emitterName=Charges}}}
    Maximum Age Spawn Threshold=1.2635
    Minimum Age Spawn Threshold=0.25
    Use Maximum Age=false
    Use Minimum Age=false
   )
  ]
  particleSpawnScript [
   &InitializeParticle(
    Color=<rgba 0.1042 0 0.0261 2>
    Sprite Rotation Mode=::NewEnumerator2
    Sprite Size Mode=::NewEnumerator1
    Uniform Sprite Size Max=40
    Uniform Sprite Size Min=25
    UsePositionOffset=false
    Write Material Random=true
   )
   &SampleParticlesFromOtherEmitter(Dynamic Material Parameter 3 Sampling:ENiagara_DirectReadSamplingMode=::NewEnumerator0, Transform Data Option=1)
   &DynamicMaterialParameters(
    Index 0 Param 1=&RandomRangeFloat(Minimum=0, Maximum=1)
    Module.Param0WriteEnabled=true
    Module.Param1WriteEnabled=true
    Module.Param2WriteEnabled=true
    Module.Param3WriteEnabled=true
   ) {module=DynamicMaterialParameters001}
   &ShapeLocation(
    Additional Quaternion Rotation=false
    MaskSpawnGroup=false
    Offset=<-5 0 0>
    Offset Coordinate Space=::Local
    Offset Mode=::NewEnumerator0
    Override Seed=false
    Rotation Mode=::NewEnumerator3
    Rotation Quaternion=/Niagara/DynamicInputs/Vectors/OrientMeshToVector(
     Look At Direction=&NormalizeVector(
      Vector To Normalize=&TransformVector(
       Vector=&MakeVector(
        X=&MakeFloatFromVector4(Vector 4=$Particles.DynamicMaterialParameter3, Channel=::NewEnumerator1)
        Y=&MakeFloatFromVector4(Vector 4=$Particles.DynamicMaterialParameter3, Channel=::NewEnumerator2)
        Z=&MakeFloatFromVector4(Vector 4=$Particles.DynamicMaterialParameter3, Channel=::NewEnumerator3)
       )
       Source Space=::World
       Destination Space=::Simulation
      )
      Safe Normalize=true
      Safe Fallback Threshold=0
      Safe Fallback Vector=<0 0 0.001>
     )
     Reference Axis Vector=<1 0 0>
    )
    Sphere Radius=2.5
    Sphere Surface Distribution=1
    Transform Order:ENiagara_TransformOrder=::NewEnumerator1
    UseOwnerScale=true
   ) {enabled=false}
   &CameraOffset()
  ]
  particleUpdateScript [
   &ParticleState()
   &InheritSourceMovement(Applied Position Scale=<0.9 0.9 1>, ApplyForce=false, ApplyPosition=true, ApplyPositionFalloff=false)
   &ScaleSpriteSize(Scale Factor=&Multiply_Vector2(
     A=<1 1>
     B=&Vector2DFromFloat(Value=&FloatFromCurve(
       FloatCurve=DataInterface() {propertyValues {Curve {keys [
           {arriveTangent=24.2191, interpMode=RCIM_Cubic, leaveTangent=24.2191, tangentMode=RCTM_User, time=0, value=0.5}
           {arriveTangent=1.1179, interpMode=RCIM_Cubic, leaveTangent=1.1179, tangentMode=RCTM_User, time=0.2946, value=2.5183}
           {arriveTangent=-1.9107, interpMode=RCIM_Cubic, leaveTangent=-1.9107, tangentMode=RCTM_User, time=1, value=0}
          ]}}}
       Scale Curve=&FloatFromCurve(
        FloatCurve=DataInterface() {propertyValues {Curve {keys [
            {arriveTangent=1.7033, interpMode=RCIM_Cubic, leaveTangent=1.7033, tangentMode=RCTM_User, time=0, value=0}
            {arriveTangent=2.0819, interpMode=RCIM_Cubic, leaveTangent=2.0819, tangentMode=RCTM_User, time=0.5, value=1}
           ]}}}
        CurveIndex=$Engine.System.Age
        Scale Curve=0.5
       )
      ))
    ))
   &ScaleColor(
    Linear Color Curve=DataInterface() {propertyValues {
      AlphaCurve {keys [<at 0.0078 0>,<at 0.2964 1>,<at 0.4571 1>,<at 1 0>]}
      BlueCurve {keys [<at 0.2205 4>,<at 0.5179 4.5041>,<at 0.9893 0>]}
      GreenCurve {keys [<at 0.2205 4>,<at 0.5179 5>,<at 0.9893 0>]}
      RedCurve {keys [<at 0.2205 4>,<at 0.5179 0>,<at 0.9893 0>]}
     }}
    Scale Mode=::NewEnumerator2
   ) {module=ScaleColor001}
   &GravityForce(Gravity=&RandomRangeVector(Minimum=<0 0 -60>, Maximum=<0 0 -20>))
   &SolveForcesAndVelocity(Clamp Velocity=false, Limit Acceleration=false)
   &ScaleColor(
    Scale Alpha=&OneMinusFloat(Float=&MakeFloatFromVector4(Vector 4=$Particles.DynamicMaterialParameter3, Channel=::NewEnumerator0))
    ScaleA=true
    ScaleRGB=false
   )
   &ScaleColor(
    Scale Alpha=&MakeCustomFloatFromBool(Bool=&Boolean_NotOperation(Bool=$User.blsADS), True Float=0.5, False Float=0)
    ScaleA=true
    ScaleRGB=false
   ) {module=ScaleColor002}
  ]
  rendererClasses [&NiagaraSpriteRendererProperties]
  renderers [{rendererClass=&NiagaraSpriteRendererProperties}]
 }
 Energy001 {
  emitterUpdateScript [
   &EmitterState(Inactive Response=::NewEnumerator1, Life Cycle Mode=::NewEnumerator1, Loop Duration=1, UseLoopDelay=false)
   &SpawnParticlesFromOtherEmitter(
    Attribute Reader=DataInterface() {propertyValues {EmitterBinding {bindingMode=Other, emitterName=Charges}}}
    Maximum Age Spawn Threshold=1.2635
    Minimum Age Spawn Threshold=0.25
    Spawn Rate=8
    Use Maximum Age=false
    Use Minimum Age=false
   )
  ]
  particleSpawnScript [
   &InitializeParticle(
    Color=&SelectLinearColorFromArray(
     Array Sampling Mode=::NewEnumerator0
     Direct Array Index=?
     Color Selection Array=DataInterface() {propertyValues {ColorData [<rgba 1 0 0.0739 10>, <rgba 0.5266 0 1 5>]}}
     Random Seed=0
    )
    Mesh Scale Max=&MakeCustomVectorFromBool(Bool=&RandomBool(), True Vector=<2 6 1>, False Vector=<2 2 8>)
    Mesh Scale Mode=::NewEnumerator2
    Sprite Rotation Mode=::NewEnumerator2
    Sprite Size Mode=::NewEnumerator1
    Uniform Sprite Size Max=40
    Uniform Sprite Size Min=25
    UsePositionOffset=false
    Write Material Random=true
   )
   &SampleParticlesFromOtherEmitter(Dynamic Material Parameter 3 Sampling:ENiagara_DirectReadSamplingMode=::NewEnumerator0, Transform Data Option=1)
   &DynamicMaterialParameters(Module.Param0WriteEnabled=true, Module.Param1WriteEnabled=true, Module.Param2WriteEnabled=true, Module.Param3WriteEnabled=true) {module=DynamicMaterialParameters001}
   &ShapeLocation(
    Additional Quaternion Rotation=false
    MaskSpawnGroup=false
    Offset=<-5 0 0>
    Offset Coordinate Space=::Local
    Offset Mode=::NewEnumerator0
    Override Seed=false
    Rotation Mode=::NewEnumerator3
    Rotation Quaternion=&OrientMeshToVector(
     Look At Direction=&NormalizeVector(
      Vector To Normalize=&TransformVector(
       Vector=&MakeVector(
        X=&MakeFloatFromVector4(Vector 4=$Particles.DynamicMaterialParameter3, Channel=::NewEnumerator1)
        Y=&MakeFloatFromVector4(Vector 4=$Particles.DynamicMaterialParameter3, Channel=::NewEnumerator2)
        Z=&MakeFloatFromVector4(Vector 4=$Particles.DynamicMaterialParameter3, Channel=::NewEnumerator3)
       )
       Source Space=::World
       Destination Space=::Simulation
      )
      Safe Normalize=true
      Safe Fallback Threshold=0
      Safe Fallback Vector=<0 0 0.001>
     )
     Reference Axis Vector=<1 0 0>
    )
    Sphere Radius=2.5
    Sphere Surface Distribution=1
    Transform Order:ENiagara_TransformOrder=::NewEnumerator1
    UseOwnerScale=true
   ) {enabled=false}
   &CameraOffset(Camera Offset Amount=4)
   &ShapeLocation(MaskSpawnGroup=false, Override Seed=false, Sphere Radius=5, UseOwnerScale=true) {module=ShapeLocation001}
  ]
  particleUpdateScript [
   &ParticleState()
   &InheritSourceMovement(ApplyForce=false, ApplyPosition=true, ApplyPositionFalloff=false)
   &ScaleMeshSize(Scale Factor=&Multiply_VectorByFloat(
     Vector=&MaskVectorBySpawnGroup(
      Masked Vector=<1 1 1>
      Passthrough Vector=&VectorFromCurve(VectorCurve=DataInterface() {propertyValues {
         XCurve {keys [<at 0 1>,<at 0.1873 0.8969>,<at 0.9772 0.8117>]}
         YCurve {keys [<at 0 1>,<at 0.1987 1.1064>,<at 0.9886 0.8379>]}
         ZCurve {keys [<at 0.0333 0.2816>,<at 0.4904 0.9738>,<at 1.0228 0.3404>]}
        }})
      Spawn Group Mask=&RandomRangeInt(Minimum=0, Maximum=1)
     )
     Float=&Multiply_Float(
      A=0.2
      B=&FloatFromCurve(
       FloatCurve=DataInterface() {propertyValues {Curve {keys [
           {arriveTangent=4.1785, time=0, value=1}
           <at 0.5 1.1>
           {arriveTangent=-0.0405, interpMode=RCIM_Cubic, leaveTangent=-0.0405, tangentMode=RCTM_User, time=1, value=1.5}
          ]}}}
       Scale Curve=0.5
      )
     )
    ))
   &ScaleColor(
    Linear Color Curve=DataInterface() {propertyValues {
      AlphaCurve {keys [<at 0.0078 0>,<at 0.2964 1>,<at 0.4571 1>,<at 1 0>]}
      BlueCurve {keys [<at 0.2205 4>,<at 0.5179 4.5041>,<at 0.9893 0>]}
      GreenCurve {keys [<at 0.2205 4>,<at 0.5179 5>,<at 0.9893 0>]}
      RedCurve {keys [<at 0.2205 4>,<at 0.5179 0>,<at 0.9893 0>]}
     }}
    Scale Mode=::NewEnumerator2
   ) {enabled=false, module=ScaleColor001}
   &ScaleColor(
    Scale Alpha=&Multiply_Float(
     A=&FloatFromCurve(
      FloatCurve=DataInterface() {propertyValues {Curve {keys [
          {arriveTangent=9.0725, interpMode=RCIM_Cubic, leaveTangent=9.0725, tangentMode=RCTM_User, time=0, value=0}
          {arriveTangent=-2.5249, interpMode=RCIM_Cubic, leaveTangent=-2.5249, tangentMode=RCTM_User, time=0.2, value=0.9847}
          {arriveTangent=-0.0405, interpMode=RCIM_Cubic, leaveTangent=-0.0405, tangentMode=RCTM_User, time=1, value=0}
         ]}}}
      Scale Curve=&RandomRangeFloat(Minimum=0.2, Maximum=1)
     )
     B=0.5
    )
    Scale RGB=&Multiply_VectorByFloat(
     Vector=&MakeVectorFromLinearColorRGB(Color=&ColorFromCurve(ColorCurve=DataInterface() {propertyValues {
         AlphaCurve {keys []}
         BlueCurve {keys [
           <at 0 5 RCIM_Cubic>
           <at 0.1326 5>
           <at 0.2869 5>
           <at 0.4134 0>
           <at 0.5108 5>
           <at 0.6236 0.0803>
           {arriveTangent=26.436, interpMode=RCIM_Cubic, leaveTangent=26.436, time=0.7107, value=0.8321}
           <at 0.8097 5>
          ]}
         GreenCurve {keys [
           <at 0 4.702 RCIM_Cubic>
           <at 0.1326 0>
           <at 0.2869 0.6611>
           <at 0.4134 0.0442>
           <at 0.5108 0.7585>
           <at 0.6236 5>
           {arriveTangent=-21.532, interpMode=RCIM_Cubic, leaveTangent=-21.532, time=0.7107, value=1}
           <at 0.8097 0.9929>
          ]}
         RedCurve {keys [<at 0 0 RCIM_Cubic>,<at 0.1326 2.465>,<at 0.2869 0>,<at 0.4134 5>,<at 0.5108 4.6237>,<at 0.6236 0>,<at 0.7107 0 RCIM_Cubic>,<at 0.8097 0>]}
        }}))
     Float=&FloatFromCurve(
      FloatCurve=DataInterface() {propertyValues {Curve {keys [{arriveTangent=-2.0684, interpMode=RCIM_Cubic, leaveTangent=-2.0684, tangentMode=RCTM_User, time=0, value=2}, <at 1 0>]}}}
      Scale Curve=&Multiply_Float(A=1, B=1)
     )
    )
    ScaleA=true
    ScaleRGB=true
   ) {module=ScaleColor003}
   &GravityForce(Gravity=&RandomRangeVector(Minimum=<0 0 -60>, Maximum=<0 0 -20>)) {enabled=false}
   &SolveForcesAndVelocity(Clamp Velocity=false, Limit Acceleration=false)
   &ScaleColor(
    Scale Alpha=&OneMinusFloat(Float=&MakeFloatFromVector4(Vector 4=$Particles.DynamicMaterialParameter3, Channel=::NewEnumerator0))
    ScaleA=true
    ScaleRGB=false
   )
   &ScaleColor(
    Scale Alpha=&MakeCustomFloatFromBool(Bool=&Boolean_NotOperation(Bool=$User.blsADS), True Float=0.5, False Float=0)
    ScaleA=true
    ScaleRGB=false
   ) {module=ScaleColor002}
  ]
  rendererClasses [&NiagaraMeshRendererProperties]
  renderers [{rendererClass=&NiagaraMeshRendererProperties, rendererIndex=0}]
 }
 ShineBurst {
  emitterUpdateScript [&EmitterState()]
  particleSpawnScript [
   &InitializeParticle(
    Color=&RandomRangeLinearColor(Minimum=<rgba 1 1 1 0.25>, Maximum=<rgba 1 1 1 0.5>)
    Lifetime Max=0.5
    Lifetime Min=0.25
    Lifetime Mode=::NewEnumerator1
    Sprite Rotation Mode=::NewEnumerator2
    Sprite Size Max=<150 100>
    Sprite Size Min=<100 75>
    Sprite Size Mode=::NewEnumerator2
    UsePositionOffset=false
    Write Material Random=true
   )
   &SubUVAnimation(End Frame=3, SubUV Animation Mode:ENiagara_SubUVLookupMode=::NewEnumerator2)
  ]
  particleUpdateScript [
   &ParticleState()
   &ScaleSpriteSize(Scale Factor=&Multiply_Vector2D_ByFloat(
     Float=$User.Burst_Size
     Vector2D=&MakeVector2D(
      X=&FloatFromCurve(
       FloatCurve=DataInterface() {propertyValues {Curve {keys [
           {interpMode=RCIM_Cubic, tangentMode=RCTM_User, time=0, value=0}
           {interpMode=RCIM_Cubic, tangentMode=RCTM_User, time=0.2, value=1}
           {tangentMode=RCTM_User, time=1, value=0.5}
          ]}}}
       Scale Curve=0.25
      )
      Y=&FloatFromCurve(
       FloatCurve=DataInterface() {propertyValues {Curve {keys [
           <at 0 0 RCIM_Cubic>
           {arriveTangent=1.5, interpMode=RCIM_Cubic, leaveTangent=1.5, time=0.2, value=0.9}
           {tangentMode=RCTM_User, time=1, value=1.5}
          ]}}}
       Scale Curve=0.25
      )
     )
    ))
   &ScaleColor(
    Scale RGB=&MakeVectorFromLinearColorRGB(Color=&LerpLinearColors(
      StartColor=&ColorFromCurve(ColorCurve=DataInterface() {propertyValues {
         AlphaCurve {keys [<at 0.1263 0.25>,<at 1 0>]}
         BlueCurve {keys [<at 0 8.3333>,<at 0.5261 1>,<at 1 0.2569>]}
         GreenCurve {keys [<at 0 20>,<at 0.5261 2>,<at 1 0.3854>]}
         RedCurve {keys [<at 0 0>,<at 0.5261 0>,<at 1 0>]}
        }})
      EndColor=&ColorFromCurve(ColorCurve=DataInterface() {propertyValues {
         AlphaCurve {keys [<at 0.1263 0.25>,<at 1 0>]}
         BlueCurve {keys [<at 0 14.8258>,<at 0.2778 1.7001>,<at 1 0.3854>]}
         GreenCurve {keys [<at 0 0.9914>,<at 0.2778 0.1808>,<at 1 0.0191>]}
         RedCurve {keys [<at 0 20>,<at 0.2778 2>,<at 1 0.3696>]}
        }})
      LerpFactor=&MakeCustomFloatFromBool(Bool=$User.ColorVariant, True Float=1, False Float=0)
     ))
    ScaleA=true
    ScaleRGB=true
   )
   /SetVariables(Particles.VisibilityTag=&MakeCustomIntFromBool(Bool=$User.blsADS, True INT=1, False INT=0)) {isSetParameters=true}
  ]
  rendererClasses [&NiagaraSpriteRendererProperties]
  renderers [{rendererClass=&NiagaraSpriteRendererProperties}]
 }
 Skull_Bits {
  emitterUpdateScript [&EmitterState()]
  particleSpawnScript [
   &InitializeParticle(
    Color=&RandomRangeLinearColor(Minimum=<rgba 0.0521 0.0031 0.0151 1>, Maximum=<rgba 1 0.0054 0.2566 1>)
    Lifetime Max=0.15
    Lifetime Min=0.1
    Lifetime Mode=::NewEnumerator1
    Mesh Scale Max=<0.05 0.05 1>
    Mesh Scale Min=<0.01 0.01 0.5>
    Mesh Scale Mode=::NewEnumerator2
    Sprite Rotation Mode=::NewEnumerator2
    Sprite Size=&Multiply_Vector2D_ByFloat(Float=0.5, Vector2D=&RandomRangeVector2D(Minimum=<3 6>, Maximum=<4 8>))
    Sprite Size Mode=::NewEnumerator4
    UsePositionOffset=false
    Write Material Random=true
   ) {module=InitializeParticle001}
   &ShapeLocation(MaskSpawnGroup=false, Non Uniform Scale=<0.3 0.3 1>, Override Seed=false, Sphere Radius=25, UseOwnerScale=true)
   &AddVelocity(
    Module.Use Falloff Distance=false
    Override Seed=false
    Velocity Mode=::NewEnumerator1
    Velocity Speed=&RandomRangeFloat(Minimum=50, Maximum=200)
   )
   &SubUVAnimation(
    "Sprite Renderer "=DataInterface() {propertyValues {SpriteRenderer=&self:Skull_Bits_5.NiagaraSpriteRendererProperties_1}}
    SubUV Animation Mode:ENiagara_SubUVLookupModeV2=::NewEnumerator2
    UseEndFrame=false
    UseStartFrame=false
   ) {enabled=false}
   &InheritVelocity(Inherited Velocity Speed Limit=1250, Speed Limit=true, UseSpeedThreshold=false)
  ]
  particleUpdateScript [
   &CurlNoiseForce(
    Noise Frequency=25
    Noise Strength=&Multiply_Float(
     A=1500
     B=&FloatFromCurve(FloatCurve=DataInterface() {propertyValues {Curve {keys [<at 0 0>,<at 0.2 1>,<at 0.3 0>]}}})
    )
    Pan Noise=false
   )
   &GravityForce(Gravity=&RandomRangeVector(Minimum=<0 0 0>, Maximum=<1 1 -980>)) {enabled=false}
   &Drag(Drag=5, Use Linear Drag=true, Use Rotational Drag=true)
   &SolveForcesAndVelocity(Clamp Velocity=false, Limit Acceleration=false)
   &ParticleState()
   &SpriteRotationRate(Rotation Rate=&Multiply_Float(A=&RandomRangeFloat(Minimum=-1, Maximum=1), B=300))
   &ScaleSpriteSize(Scale Factor=&Vector2DFromFloat(Value=&FloatFromCurve(
      FloatCurve=DataInterface() {propertyValues {Curve {keys [
          {arriveTangent=2.5076, interpMode=RCIM_Cubic, leaveTangent=2.5076, tangentMode=RCTM_User, time=0.0023, value=0}
          {arriveTangent=0.0897, interpMode=RCIM_Cubic, leaveTangent=0.0897, tangentMode=RCTM_User, time=0.2, value=1}
          {arriveTangent=-1.8351, interpMode=RCIM_Cubic, tangentMode=RCTM_User, time=1, value=0}
         ]}}}
      Scale Curve=&Multiply_Float(A=1, B=&RandomRangeFloat(Minimum=0.25, Maximum=1))
     )))
   &ScaleColor(ScaleA=true, ScaleRGB=false)
   /SetVariables(Particles.VisibilityTag=&MakeCustomIntFromBool(Bool=$User.blsADS, True INT=1, False INT=0)) {isSetParameters=true}
  ]
  rendererClasses [&NiagaraSpriteRendererProperties, &NiagaraMeshRendererProperties]
  renderers [
   {rendererClass=&NiagaraSpriteRendererProperties}
   {rendererClass=&NiagaraMeshRendererProperties, rendererIndex=1}
  ]
 }
}
summary {
 systemName=NS_CosmicThunder_PhaseDash_Charges
 emitters [
  {emitterName=Charges, rendererClasses [&NiagaraMeshRendererProperties, &NiagaraSpriteRendererProperties]}
  {bEnabled=false, emitterName=Energy, rendererClasses [&NiagaraSpriteRendererProperties]}
  {emitterName=Skull_Bits, rendererClasses [&NiagaraSpriteRendererProperties, &NiagaraMeshRendererProperties]}
  {emitterName=ShineBurst, rendererClasses [&NiagaraSpriteRendererProperties]}
  {emitterName=Energy001, rendererClasses [&NiagaraMeshRendererProperties]}
 ]
 userVariables [
  User.blsADS {type=NiagaraBool}
  User.blsDBNO {type=NiagaraBool}
  User.Burst_Size {type=NiagaraFloat, defaultValue=1, description=""}
  User.ColorVariant {type=NiagaraBool, defaultValue=true}
  User.MaxNumCharges {type=NiagaraInt32, defaultValue=3}
  User.NumCharges {type=NiagaraInt32, defaultValue=3}
  User.Skull_Charge_Counter {type=NiagaraInt32}
 ]
}
