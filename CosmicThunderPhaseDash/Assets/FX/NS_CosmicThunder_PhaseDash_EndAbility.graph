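# Niagara effect system for the Cosmic Thunder Phase Dash, named as the effect for the end of the ability.
# It holds six emitters: three GlitchesWide variants, a lightweight smoke wisp and two generically named Empty emitters.
# System logic consists of System State plus a Set Variables assignment from the spawn script.
# No user variables are exposed, and no renderers are listed for any emitter.
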NiagaraSystem /CosmicThunderPhaseDash/Assets/FX/NS_CosmicThunder_PhaseDash_EndAbility
dependencies {
 usedDataInterfaces []
 usedDynamicInputs []
 usedModules [
  &self:SystemSpawnScript.SystemScriptSource.SystemScriptGraph.NiagaraNodeAssignment_0.SetVariables_#1
  &SystemState
 ]
}
emitters {
 Empty001 {rendererClasses [], renderers []}
 Empty002 {rendererClasses [], renderers []}
 GlitchesWide {rendererClasses [], renderers []}
 GlitchesWide001 {rendererClasses [], renderers []}
 GlitchesWide002 {rendererClasses [], renderers []}
 Smoke_Wisp_Lightweight002 {rendererClasses [], renderers []}
}
summary {
 systemName=NS_CosmicThunder_PhaseDash_EndAbility
 emitters [
  {emitterName=Empty001, rendererClasses []}
  {emitterName=Empty002, rendererClasses []}
  {emitterName=Smoke_Wisp_Lightweight002, rendererClasses []}
  {emitterName=GlitchesWide, rendererClasses []}
  {emitterName=GlitchesWide001, rendererClasses []}
  {emitterName=GlitchesWide002, rendererClasses []}
 ]
 userVariables []
}
