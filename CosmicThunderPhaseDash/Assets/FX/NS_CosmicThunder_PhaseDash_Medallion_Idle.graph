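# Niagara system for the idle visual effect of the Cosmic Thunder Phase Dash medallion, judging by its name.
# It layers named emitters for shafts, a light ray, sparkles, and long and wide glitches, plus a generically named Empty001 and a disabled Empty.
# Only the SystemState module is listed, with no user variables, data interfaces or dynamic inputs.
# Emitter renderer lists are empty here, so the file does not show how each layer draws.
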NiagaraSystem /CosmicThunderPhaseDash/Assets/FX/NS_CosmicThunder_PhaseDash_Medallion_Idle
dependencies {usedDataInterfaces [], usedDynamicInputs [], usedModules [&SystemState]}
emitters {
 Empty {bEnabled=false, rendererClasses [], renderers []}
 Empty001 {rendererClasses [], renderers []}
 GlitchesLong {rendererClasses [], renderers []}
 GlitchesWide {rendererClasses [], renderers []}
 LghtRay {rendererClasses [], renderers []}
 Shafts {rendererClasses [], renderers []}
 Sparkles {rendererClasses [], renderers []}
}
summary {
 systemName=NS_CosmicThunder_PhaseDash_Medallion_Idle
 emitters [
  {emitterName=Shafts, rendererClasses []}
  {emitterName=Sparkles, rendererClasses []}
  {emitterName=LghtRay, rendererClasses []}
  {bEnabled=false, emitterName=Empty, rendererClasses []}
  {emitterName=Empty001, rendererClasses []}
  {emitterName=GlitchesLong, rendererClasses []}
  {emitterName=GlitchesWide, rendererClasses []}
 ]
 userVariables []
}
