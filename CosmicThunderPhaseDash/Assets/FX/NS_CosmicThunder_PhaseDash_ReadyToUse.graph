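# Niagara system for the Cosmic Thunder Phase Dash "ReadyToUse" effect, wrapping a skeletal mesh in pink-purple spirit energy.
# Energy_Spirit mixes a burst with loop-count-gated rate spawning to feed sprite and mesh renderers, using two spawn groups for mesh orientation.
# Particles resample the mesh each update with a vertical offset, while curves over particle age, system age and normalized velocity shape color and opacity.
# Three GlitchesWide emitters show no renderers or modules.
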NiagaraSystem /CosmicThunderPhaseDash/Assets/FX/NS_CosmicThunder_PhaseDash_ReadyToUse
dependencies {
 usedDataInterfaces [&NiagaraDataInterfaceCurve, &NiagaraDataInterfaceSkeletalMesh]
 usedDynamicInputs [
  &ClampFloat
  &RemapRange
  &VectorLength
  &SetBoolByFloatComparison
 ]
 usedModules [
  &self:SystemSpawnScript.SystemScriptSource.SystemScriptGraph.NiagaraNodeAssignment_0.SetVariables_#1
  &SystemState
  &InterpolateOverTime
  &self:SystemSpawnScript.SystemScriptSource.SystemScriptGraph.NiagaraNodeAssignment_1.SetVariables_#2
  &self:SystemSpawnScript.SystemScriptSource.SystemScriptGraph.NiagaraNodeAssignment_2.SetVariables_#3
  &Timeline
 ]
}
emitters {
 Energy_Spirit {
  emitterUpdateScript [
   &EmitterState(
    Enable Distance Culling=true
    Inactive Response=::NewEnumerator1
    Life Cycle Mode=::NewEnumerator1
    Loop Duration=0.5
    MaxDistance=17000
    Module.bUseMaxDistance=true
    Module.bUseMinDistance=false
    Scalability Mode=::NewEnumerator1
    UseLoopDelay=true
   )
   &SpawnBurst_Instantaneous(Spawn Count=4, Spawn Time=0.1, Use Spawn Probability=false, UseLoopCountLimit=true) {module=SpawnBurst_Instantaneous001}
   &SpawnRate(
    Spawn Probability=&MakeCustomFloatFromBool(
     Bool=&SetBoolByIntComparison(A=$Emitter.LoopCount, B=1, Comparison Type=::NewEnumerator6)
     True Float=0
     False Float=1
    )
    SpawnRate=8
    Use Spawn Probability=true
   )
  ]
  particleSpawnScript [
   &FN_SetSpawnGroup(Number of Spawn Groups=2)
   &InitializeParticle(
    Color=&RandomRangeLinearColor(Minimum=<rgba 0.9217 0.0786 0.5011 1.5>, Maximum=<rgba 0.0409 0 0.2396 2>)
    Lifetime=&MakeCustomFloatFromBool(
     Bool=&SetBoolByIntComparison(A=$Emitter.LoopCount, B=1, Comparison Type=::NewEnumerator6)
     True Float=&RandomRangeFloat(Minimum=1, Maximum=1.5)
     False Float=&RandomRangeFloat(Minimum=1, Maximum=1.5)
    )
    Mesh Scale Max=<1.5 1.5 1.5>
    Mesh Scale Min=<0.5 0.5 0.5>
    Mesh Scale Mode=::NewEnumerator2
    Sprite Rotation Mode=::NewEnumerator2
    Sprite Size Max=<35 45>
    Sprite Size Min=<30 40>
    Sprite Size Mode=::NewEnumerator2
    Sprite UV Mode=::NewEnumerator3
    UsePositionOffset=false
    Write Material Random=true
   )
   /SetVariables() {isSetParameters=true}
   &DynamicMaterialParameters(
    Index 0 Param 1=&RandomRangeFloat(Minimum=0, Maximum=1, Evaluation Type=::NewEnumerator1)
    Index 0 Param 2=&RandomRangeFloat(Minimum=0, Maximum=1, Evaluation Type=::NewEnumerator1)
    Module.Param0WriteEnabled=true
    Module.Param1WriteEnabled=true
    Module.Param2WriteEnabled=true
    Module.Param3WriteEnabled=true
   ) {module=DynamicMaterialParameters002}
   &CameraOffset(Camera Offset Amount=25)
   &InitialMeshOrientation(
    Mesh Axis To Orient=&MaskVectorBySpawnGroup(
     Masked Vector=<0 1 0>
     Passthrough Vector=<1 0 0>
     Spawn Group Mask=&RandomRangeInt(Minimum=0, Maximum=1)
    )
    Mesh Orientation Mode=::NewEnumerator2
    Orientation Vector=&MaskVectorBySpawnGroup(
     Masked Vector=<0 1 0>
     Passthrough Vector=<1 0 0>
     Spawn Group Mask=&RandomRangeInt(Minimum=0, Maximum=1)
    )
    Rotation=&MaskVectorBySpawnGroup(
     Masked Vector=<0 0.5 0>
     Passthrough Vector=<0 0 1>
     Spawn Group Mask=&RandomRangeInt(Minimum=0, Maximum=1)
    )
    Use Rotation Vector=true
   )
  ]
  particleUpdateScript [
   &ParticleState()
   &ScaleSpriteSize(Scale Sprite Size Mode=::NewEnumerator0, Uniform Scale Factor=0.75)
   /SetVariables(
    Particles.Color_Sprite_Dark:LinearColor=&MultiplyLinearColor(
     Linear Color=<rgba 0 0 0 0.5>
     Scale Factor=&MakeFloat4FromVectorAndFloat(
      "Vector (RGB)"=<1 1 1>
      "Float (Alpha)"=&Multiply_Float(
       A=&FloatFromCurve(
        FloatCurve=DataInterface() {propertyValues {Curve {keys [
            {arriveTangent=3.2889, interpMode=RCIM_Cubic, leaveTangent=3.2889, tangentMode=RCTM_User, time=0, value=0}
            {interpMode=RCIM_Cubic, tangentMode=RCTM_User, time=1, value=1}
           ]}}}
        CurveIndex=$Engine.System.Age
       )
       B=&FloatFromCurve(FloatCurve=DataInterface() {propertyValues {Curve {
           defaultValue=1
           keys [
            {interpMode=RCIM_Cubic, tangentMode=RCTM_User, time=0, value=0}
            {interpMode=RCIM_Cubic, tangentMode=RCTM_User, time=0.5, value=1}
            {tangentMode=RCTM_User, time=1, value=0}
           ]
          }}})
      )
     )
    )
    Particles.Mesh Color:LinearColor=&MultiplyLinearColor(
     Linear Color=&MakeCustomLinearColorFromBool(
      Bool=&RandomBool()
      True Color=&MakeCustomLinearColorFromBool(Bool=&RandomBool(), True Color=<rgba 0.8 0 0.3189 1>, False Color=<rgba 0.0514 0 0.4062 1>)
      False Color=<rgba 0.2157 0 0.5 1>
     )
     Scale Factor=&MakeVector4(
      X=1
      Y=1
      Z=1
      W=&FloatFromCurve(FloatCurve=DataInterface() {propertyValues {Curve {keys [<at 0.3 1 RCIM_Cubic>,<at 1 0>]}}}, Scale Curve=2)
     )
    )
    Particles.Sprite_Size_Dark:Vector2f=&Multiply_Vector2D_ByFloat(
     Float=&FloatFromCurve(FloatCurve=DataInterface() {propertyValues {Curve {keys [
          {interpMode=RCIM_Cubic, tangentMode=RCTM_User, time=0, value=0.75}
          {interpMode=RCIM_Cubic, tangentMode=RCTM_User, time=0.5, value=1}
          {tangentMode=RCTM_User, time=1.0123, value=0.8215}
         ]}}})
     Vector2D=<70 70>
    )
   ) {isSetParameters=true}
   &SkeletalMeshLocation(
    Bone Sampling Mode=::NewEnumerator0
    "Interpolate Mesh Sampling (Update)"=true
    MaskSpawnGroup=false
    Orientation Sampling=::NewEnumerator1
    OverrideRandomSeed=false
    Sampled Position Offset=&AddVectorToPosition(
     Position=&RandomRangeVector(Minimum=<0 0 -15>, Maximum=<0 0 5>)
     Vector=&MakeVector(
      X=1
      Y=1
      Z=&FloatFromCurve(
       FloatCurve=DataInterface() {propertyValues {Curve {keys [
           {arriveTangent=1.9662, interpMode=RCIM_Cubic, leaveTangent=1.9662, tangentMode=RCTM_User, time=0, value=0}
           {interpMode=RCIM_Cubic, tangentMode=RCTM_User, time=1, value=1}
          ]}}}
       Scale Curve=20
      )
     )
    )
    Sampled Position Offset Coordinate Space:ENiagaraSkelMeshTransforms=::NewEnumerator2
    Skeletal Mesh=$System.Skeletal Mesh
   ) {module=SkeletalMeshLocation001}
   &SpriteRotationRate(
    Delta Time:NiagaraFloat=$Particles.NormalizedAge
    Rotation Rate=&RandomRangeFloat(Minimum=-0.5, Maximum=0.5)
   )
   &ScaleColor(
    Scale Alpha=&Multiply_Float(
     A=&Multiply_Float(
      A=&FloatFromCurve(
       FloatCurve=DataInterface() {propertyValues {Curve {keys [
           {arriveTangent=4.0047, interpMode=RCIM_Cubic, leaveTangent=4.0047, tangentMode=RCTM_User, time=0, value=0}
           {arriveTangent=-1.9403, interpMode=RCIM_Cubic, leaveTangent=-1.9403, tangentMode=RCTM_User, time=0.3313, value=0.9311}
           {tangentMode=RCTM_User, time=1, value=0}
          ]}}}
       Scale Curve=&RandomRangeFloat(Minimum=0.7, Maximum=1.2)
      )
      B=0.5
     )
     B=&FloatFromCurve(
      FloatCurve=DataInterface() {propertyValues {Curve {keys [
          {arriveTangent=0.0152, interpMode=RCIM_Cubic, leaveTangent=0.0152, tangentMode=RCTM_User, time=0, value=1}
          {arriveTangent=-0.3197, interpMode=RCIM_Cubic, leaveTangent=-0.3197, tangentMode=RCTM_User, time=1, value=0.75}
         ]}}}
      CurveIndex=$System.NormalizedVelocityRange
      Scale Curve=&FloatFromCurve(
       FloatCurve=DataInterface() {propertyValues {Curve {keys [{interpMode=RCIM_Cubic, tangentMode=RCTM_User, time=0, value=1}, {interpMode=RCIM_Cubic, tangentMode=RCTM_User, time=1, value=0}]}}}
       Scale Curve=1.5
      )
     )
    )
    Scale RGB=&Multiply_VectorByFloat(
     Vector=<1 1 1>
     Float=&FloatFromCurve(FloatCurve=DataInterface() {propertyValues {Curve {keys [<at 0 1>,<at 0.5 0.3>]}}}, CurveIndex=$Engine.System.Age, Scale Curve=5)
    )
    ScaleA=true
    ScaleRGB=true
   ) {module=ScaleColor002}
   &SolveForcesAndVelocity(Clamp Velocity=false, Limit Acceleration=false)
   &DynamicMaterialParameters(
    Index 0 Param 1=&Add_Float(
     A=&RandomRangeFloat(Minimum=0, Maximum=1)
     B=&Multiply_Float(A=$Particles.NormalizedAge, B=1)
    )
    Module.Index0Float0Write=true
    Module.Index0Float1Write=false
    Module.Index0Float2Write=false
    Module.Index0Float3Write=false
   ) {enabled=false}
  ]
  rendererClasses [&NiagaraSpriteRendererProperties, &NiagaraMeshRendererProperties]
  renderers [
   {rendererClass=&NiagaraSpriteRendererProperties}
   {rendererClass=&NiagaraMeshRendererProperties, rendererIndex=1}
  ]
 }
 GlitchesWide {rendererClasses [], renderers []}
 GlitchesWide001 {rendererClasses [], renderers []}
 GlitchesWide002 {rendererClasses [], renderers []}
}
summary {
 systemName=NS_CosmicThunder_PhaseDash_ReadyToUse
 emitters [
  {emitterName=Energy_Spirit, rendererClasses [&NiagaraSpriteRendererProperties, &NiagaraMeshRendererProperties]}
  {emitterName=GlitchesWide, rendererClasses []}
  {emitterName=GlitchesWide001, rendererClasses []}
  {emitterName=GlitchesWide002, rendererClasses []}
 ]
 userVariables []
}
